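# Animation Blueprint for the AloeCrouton gun mesh, driving its idle, charge, fire and cooldown animations.
# Event graph caches weapon charge and equip state, pawn movement flags and the character's BothHandFK curve each update.
# One state machine feeds DefaultSlot, entering cooldown when charge count is zero and re-equipping when it recovers.
# Global state aliases force a scaled-down pose for unequipped, holstered, prop-entry, zipline and swimming cases.
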AnimBlueprint /AloeCroutonWeapon/Items/AloeCrouton/Animation/AloeCroutonAnimBlueprint
target_skeleton=/AloeCroutonWeapon/Items/AloeCrouton/Meshes/AloeCrouton_Gun_Skeleton
graphs [
 AnimGraph {nodes [
   AnimGraphNode_StateMachine_0 = AloeCrouton_SM() {is=title AnimGraphNode_StateMachine}
   Output Pose(Result=Slot 'DefaultSlot'(Source=@AnimGraphNode_StateMachine_0)) {layerGroup=None}
  ]}
 AloeCrouton_SM {
  entry=Scaled Down Pose
  path=AnimGraph.AnimGraphNode_StateMachine_0.AloeCrouton_SM
  ids=id
  nodes [
   Alias to Sprint Swimming() {is=title AnimStateAliasNode, bGlobalAlias=true}
   Alias to Not Equipped or Holstered() {is=title AnimStateAliasNode, bGlobalAlias=true}
   Alias to Vertical Zipline or Unequiped Horizontal Zipline() {is=title AnimStateAliasNode, bGlobalAlias=true}
   AnimStateNode_0 = Equip() {
    is=title AnimStateNode
    stateEntered {guid=#1}
    stateFullyBlended {guid=#2}
    stateLeft {guid=#3}
   }
   AnimStateNode_1 = To Cool Down() {
    is=title AnimStateNode
    stateEntered {guid=#4}
    stateFullyBlended {guid=#5}
    stateLeft {guid=#6}
   }
   AnimStateNode_10 = Zipline - Scaled Down Pose() {is=title AnimStateNode}
   AnimStateNode_11 = Empty?() {is=title AnimStateNode}
   AnimStateNode_2 = Scaled Down Pose() {is=title AnimStateNode}
   AnimStateNode_3 = Cool Down Idle() {
    is=title AnimStateNode
    stateEntered {guid=#7}
    stateFullyBlended {guid=#8}
    stateLeft {guid=#9}
   }
   AnimStateNode_4 = Cool Down Over() {is=title AnimStateNode}
   AnimStateNode_5 = Idle() {
    stateEntered {guid=#10}
    stateFullyBlended {guid=#11}
    stateLeft {guid=#12}
   }
   AnimStateNode_6 = Charge Loop() {is=title AnimStateNode}
   AnimStateNode_7 = Charge Start() {is=title AnimStateNode}
   AnimStateNode_8 = Fire() {
    is=title AnimStateNode
    stateEntered {guid=#13}
    stateFullyBlended {guid=#14}
    stateLeft {guid=#15}
   }
   AnimStateNode_9 = Sprint Swimming - Scaled Down Pose() {is=title AnimStateNode}
   AnimStateTransitionNode_0 = AnimStateTransitionNode() {crossfadeDuration=0}
   AnimStateTransitionNode_1 = AnimStateTransitionNode() {
    priorityOrder=100
    transitionEnd {guid=#16}
    transitionInterrupt {guid=#17}
    transitionStart {guid=#18}
   }
   AnimStateTransitionNode_10 = AnimStateTransitionNode() {bAutomaticRuleBasedOnSequencePlayerInState=true}
   AnimStateTransitionNode_11 = AnimStateTransitionNode()
   AnimStateTransitionNode_12 = AnimStateTransitionNode()
   AnimStateTransitionNode_13 = AnimStateTransitionNode()
   AnimStateTransitionNode_17 = AnimStateTransitionNode() {crossfadeDuration=0}
   AnimStateTransitionNode_2 = AnimStateTransitionNode() {bAutomaticRuleBasedOnSequencePlayerInState=true}
   AnimStateTransitionNode_20 = AnimStateTransitionNode() {crossfadeDuration=0}
   AnimStateTransitionNode_21 = AnimStateTransitionNode() {crossfadeDuration=0}
   AnimStateTransitionNode_23 = AnimStateTransitionNode() {crossfadeDuration=0}
   AnimStateTransitionNode_24 = AnimStateTransitionNode() {crossfadeDuration=0}
   AnimStateTransitionNode_25 = AnimStateTransitionNode() {crossfadeDuration=0}
   AnimStateTransitionNode_26 = AnimStateTransitionNode() {crossfadeDuration=0}
   AnimStateTransitionNode_27 = AnimStateTransitionNode() {crossfadeDuration=0}
   AnimStateTransitionNode_3 = AnimStateTransitionNode() {crossfadeDuration=0}
   AnimStateTransitionNode_4 = AnimStateTransitionNode() {crossfadeDuration=0}
   AnimStateTransitionNode_5 = AnimStateTransitionNode() {bAutomaticRuleBasedOnSequencePlayerInState=true}
   AnimStateTransitionNode_6 = AnimStateTransitionNode()
   AnimStateTransitionNode_7 = AnimStateTransitionNode()
   AnimStateTransitionNode_8 = AnimStateTransitionNode() {bAutomaticRuleBasedOnSequencePlayerInState=true}
   AnimStateTransitionNode_9 = AnimStateTransitionNode()
  ]
  links [
   {
    dst=Scaled Down Pose
    rule=AnimGraph.AnimGraphNode_StateMachine_0.AloeCrouton_SM.AnimStateTransitionNode_3.Transition
    src=Alias to Not Equipped or Holstered
    via=AnimStateTransitionNode_3
   }
   {
    dst=Sprint Swimming - Scaled Down Pose
    rule=AnimGraph.AnimGraphNode_StateMachine_0.AloeCrouton_SM.AnimStateTransitionNode_26.Transition
    src=Alias to Sprint Swimming
    via=AnimStateTransitionNode_26
   }
   {
    dst=Zipline - Scaled Down Pose
    rule=AnimGraph.AnimGraphNode_StateMachine_0.AloeCrouton_SM.AnimStateTransitionNode_21.Transition
    src=Alias to Vertical Zipline or Unequiped Horizontal Zipline
    via=AnimStateTransitionNode_21
   }
   {
    dst=Fire
    rule=AnimGraph.AnimGraphNode_StateMachine_0.AloeCrouton_SM.AnimStateTransitionNode_17.Transition
    src=Charge Loop
    via=AnimStateTransitionNode_17
   }
   {
    dst=Charge Loop
    rule=AnimGraph.AnimGraphNode_StateMachine_0.AloeCrouton_SM.AnimStateTransitionNode_10.AnimationTransitionGraph_0
    src=Charge Start
    via=AnimStateTransitionNode_10
   }
   {
    dst=Idle
    rule=AnimGraph.AnimGraphNode_StateMachine_0.AloeCrouton_SM.AnimStateTransitionNode_11.AnimationTransitionGraph_0
    src=Charge Start
    via=AnimStateTransitionNode_11
   }
   {
    dst=Cool Down Over
    rule=AnimGraph.AnimGraphNode_StateMachine_0.AloeCrouton_SM.AnimStateTransitionNode_7.Transition
    src=Cool Down Idle
    via=AnimStateTransitionNode_7
   }
   {
    dst=Idle
    rule=AnimGraph.AnimGraphNode_StateMachine_0.AloeCrouton_SM.AnimStateTransitionNode_12.Transition
    src=Cool Down Over
    via=AnimStateTransitionNode_12
   }
   {
    dst=Idle
    rule=AnimGraph.AnimGraphNode_StateMachine_0.AloeCrouton_SM.AnimStateTransitionNode_8.Transition
    src=Cool Down Over
    via=AnimStateTransitionNode_8
   }
   {
    dst=Cool Down Idle
    rule=AnimGraph.AnimGraphNode_StateMachine_0.AloeCrouton_SM.AnimStateTransitionNode_13.Transition
    src=Empty?
    via=AnimStateTransitionNode_13
   }
   {
    dst=Idle
    rule=AnimGraph.AnimGraphNode_StateMachine_0.AloeCrouton_SM.AnimStateTransitionNode_6.Transition
    src=Empty?
    via=AnimStateTransitionNode_6
   }
   {
    dst=Idle
    rule=AnimGraph.AnimGraphNode_StateMachine_0.AloeCrouton_SM.AnimStateTransitionNode_0.Transition
    src=Equip
    via=AnimStateTransitionNode_0
   }
   {
    dst=Zipline - Scaled Down Pose
    rule=AnimGraph.AnimGraphNode_StateMachine_0.AloeCrouton_SM.AnimStateTransitionNode_24.Transition
    src=Equip
    via=AnimStateTransitionNode_24
   }
   {
    dst=Empty?
    rule=AnimGraph.AnimGraphNode_StateMachine_0.AloeCrouton_SM.AnimStateTransitionNode_5.Transition
    src=Fire
    via=AnimStateTransitionNode_5
   }
   {
    dst=Charge Start
    rule=AnimGraph.AnimGraphNode_StateMachine_0.AloeCrouton_SM.AnimStateTransitionNode_9.AnimationTransitionGraph_0
    src=Idle
    via=AnimStateTransitionNode_9
   }
   {
    dst=To Cool Down
    rule=AnimGraph.AnimGraphNode_StateMachine_0.AloeCrouton_SM.AnimStateTransitionNode_1.Transition
    src=Idle
    via=AnimStateTransitionNode_1
   }
   {
    dst=Equip
    rule=AnimGraph.AnimGraphNode_StateMachine_0.AloeCrouton_SM.AnimStateTransitionNode_4.Transition
    src=Scaled Down Pose
    via=AnimStateTransitionNode_4
   }
   {
    dst=Equip
    rule=AnimGraph.AnimGraphNode_StateMachine_0.AloeCrouton_SM.AnimStateTransitionNode_27.Transition
    src=Sprint Swimming - Scaled Down Pose
    via=AnimStateTransitionNode_27
   }
   {
    dst=Idle
    rule=AnimGraph.AnimGraphNode_StateMachine_0.AloeCrouton_SM.AnimStateTransitionNode_25.Transition
    src=Sprint Swimming - Scaled Down Pose
    via=AnimStateTransitionNode_25
   }
   {
    dst=Cool Down Idle
    rule=AnimGraph.AnimGraphNode_StateMachine_0.AloeCrouton_SM.AnimStateTransitionNode_2.Transition
    src=To Cool Down
    via=AnimStateTransitionNode_2
   }
   {
    dst=Equip
    rule=AnimGraph.AnimGraphNode_StateMachine_0.AloeCrouton_SM.AnimStateTransitionNode_23.Transition
    src=Zipline - Scaled Down Pose
    via=AnimStateTransitionNode_23
   }
   {
    dst=Idle
    rule=AnimGraph.AnimGraphNode_StateMachine_0.AloeCrouton_SM.AnimStateTransitionNode_20.Transition
    src=Zipline - Scaled Down Pose
    via=AnimStateTransitionNode_20
   }
  ]
 }
 Equip {
  path=AnimGraph.AnimGraphNode_StateMachine_0.AloeCrouton_SM.AnimStateNode_0.Equip
  nodes [Output Animation Pose(Result=Local Space Ref Pose())]
 }
 To Cool Down {
  path=AnimGraph.AnimGraphNode_StateMachine_0.AloeCrouton_SM.AnimStateNode_1.To Cool Down
  nodes [Output Animation Pose(Result=Sequence Player 'AloeCrouton_ToCooldown_Gun'() {bLoopAnimation=false, sequence=/AloeCroutonWeapon/Items/AloeCrouton/Animation/Cooldown/AloeCrouton_ToCooldown_Gun})]
 }
 Zipline - Scaled Down Pose {
  path=AnimGraph.AnimGraphNode_StateMachine_0.AloeCrouton_SM.AnimStateNode_10.Zipline - Scaled Down Pose
  nodes [Output Animation Pose(Result=Sequence Player 'AloeCrouton_EquipStartPose_Gun'() {bLoopAnimation=false, sequence=/AloeCroutonWeapon/Items/AloeCrouton/Animation/Equip/AloeCrouton_EquipStartPose_Gun})]
 }
 Empty? {
  path=AnimGraph.AnimGraphNode_StateMachine_0.AloeCrouton_SM.AnimStateNode_11.Empty?
  nodes [Output Animation Pose(Result=Local Space Ref Pose())]
 }
 Scaled Down Pose {
  path=AnimGraph.AnimGraphNode_StateMachine_0.AloeCrouton_SM.AnimStateNode_2.Scaled Down Pose
  nodes [Output Animation Pose(Result=Sequence Player 'AloeCrouton_EquipStartPose_Gun'() {bLoopAnimation=false, sequence=&AloeCrouton_EquipStartPose_Gun})]
 }
 Cool Down Idle {
  path=AnimGraph.AnimGraphNode_StateMachine_0.AloeCrouton_SM.AnimStateNode_3.Cool Down Idle
  nodes [Output Animation Pose(Result=Sequence Player 'AloeCrouton_Cooldown_Idle_Gun'() {sequence=/AloeCroutonWeapon/Items/AloeCrouton/Animation/Cooldown/AloeCrouton_Cooldown_Idle_Gun})]
 }
 Cool Down Over {
  path=AnimGraph.AnimGraphNode_StateMachine_0.AloeCrouton_SM.AnimStateNode_4.Cool Down Over
  nodes [Output Animation Pose(Result=Sequence Player 'AloeCrouton_Cooldown_ReEquip_Gun'() {bLoopAnimation=false, sequence=/AloeCroutonWeapon/Items/AloeCrouton/Animation/Cooldown/AloeCrouton_Cooldown_ReEquip_Gun})]
 }
 Idle {
  path=AnimGraph.AnimGraphNode_StateMachine_0.AloeCrouton_SM.AnimStateNode_5.Idle
  nodes [Output Animation Pose(Result=Sequence Player 'AloeCrouton_Idle_Gun'() {sequence=/AloeCroutonWeapon/Items/AloeCrouton/Animation/Idle/AloeCrouton_Idle_Gun})]
 }
 Charge Loop {
  path=AnimGraph.AnimGraphNode_StateMachine_0.AloeCrouton_SM.AnimStateNode_6.Charge Loop
  nodes [Output Animation Pose(Result=Sequence Player 'AloeCrouton_ChargeLoop_Gun'() {sequence=/AloeCroutonWeapon/Items/AloeCrouton/Animation/Attack/AloeCrouton_ChargeLoop_Gun})]
 }
 Charge Start {
  path=AnimGraph.AnimGraphNode_StateMachine_0.AloeCrouton_SM.AnimStateNode_7.Charge Start
  nodes [Output Animation Pose(Result=Sequence Player 'AloeCrouton_ChargeStart_Gun'() {bLoopAnimation=false, sequence=/AloeCroutonWeapon/Items/AloeCrouton/Animation/Attack/AloeCrouton_ChargeStart_Gun})]
 }
 Fire {
  path=AnimGraph.AnimGraphNode_StateMachine_0.AloeCrouton_SM.AnimStateNode_8.Fire
  nodes [Output Animation Pose(Result=Sequence Player 'AloeCrouton_Fire_Gun'() {bLoopAnimation=false, sequence=/AloeCroutonWeapon/Items/AloeCrouton/Animation/Attack/AloeCrouton_Fire_Gun})]
 }
 Sprint Swimming - Scaled Down Pose {
  path=AnimGraph.AnimGraphNode_StateMachine_0.AloeCrouton_SM.AnimStateNode_9.Sprint Swimming - Scaled Down Pose
  nodes [Output Animation Pose(Result=Sequence Player 'AloeCrouton_EquipStartPose_Gun'() {bLoopAnimation=false, sequence=&AloeCrouton_EquipStartPose_Gun})]
 }
 Transition {path=AnimGraph.AnimGraphNode_StateMachine_0.AloeCrouton_SM.AnimStateTransitionNode_0.Transition, nodes [Result(bCanEnterTransition=true)]}
 Transition {
  path=AnimGraph.AnimGraphNode_StateMachine_0.AloeCrouton_SM.AnimStateTransitionNode_1.Transition
  nodes [Result(bCanEnterTransition="Equal (Float)"(A=Get Charge Count() -> (Charge Count:double)))]
 }
 AnimationTransitionGraph_0 {path=AnimGraph.AnimGraphNode_StateMachine_0.AloeCrouton_SM.AnimStateTransitionNode_10.AnimationTransitionGraph_0, nodes [Result()]}
 AnimationTransitionGraph_0 {
  path=AnimGraph.AnimGraphNode_StateMachine_0.AloeCrouton_SM.AnimStateTransitionNode_11.AnimationTransitionGraph_0
  nodes [Result(bCanEnterTransition=NOT Boolean(A=Get IsCharging() -> (IsCharging:bool)))]
 }
 Transition {
  path=AnimGraph.AnimGraphNode_StateMachine_0.AloeCrouton_SM.AnimStateTransitionNode_12.Transition
  nodes [Result(bCanEnterTransition=Get IsCharging())]
 }
 Transition {
  path=AnimGraph.AnimGraphNode_StateMachine_0.AloeCrouton_SM.AnimStateTransitionNode_13.Transition
  nodes [Result(bCanEnterTransition="Equal (Float)"(A=Get Charge Count()))]
 }
 Transition {
  path=AnimGraph.AnimGraphNode_StateMachine_0.AloeCrouton_SM.AnimStateTransitionNode_17.Transition
  nodes [Result(bCanEnterTransition=NOT Boolean(A=Get IsCharging()))]
 }
 Transition {path=AnimGraph.AnimGraphNode_StateMachine_0.AloeCrouton_SM.AnimStateTransitionNode_2.Transition, nodes [Result()]}
 Transition {
  path=AnimGraph.AnimGraphNode_StateMachine_0.AloeCrouton_SM.AnimStateTransitionNode_20.Transition
  nodes [Result(bCanEnterTransition=NOT Boolean(A=Get IsZiplining() -> (IsZiplining:bool)))]
 }
 Transition {
  path=AnimGraph.AnimGraphNode_StateMachine_0.AloeCrouton_SM.AnimStateTransitionNode_21.Transition
  nodes [
   n1 = NOT Boolean(A=Get IsEquipping() -> (IsEquipping:bool))
   n2 = NOT Boolean(A=Get IsEquipped() -> (IsEquipped:bool))
   n3 = Get IsZiplining()
   n4 = Get IsHolstered() -> (IsHolstered:bool)
   Result(bCanEnterTransition=OR Boolean(A=AND Boolean!(A=@n1, B=@n2, C:bool=@n3) -> (ReturnValue), B=AND Boolean!(A=@n3, B=@n4) -> (ReturnValue)))
  ]
 }
 Transition {
  path=AnimGraph.AnimGraphNode_StateMachine_0.AloeCrouton_SM.AnimStateTransitionNode_23.Transition
  nodes [
   n1 = Get IsHolstered()
   Result(bCanEnterTransition=AND Boolean(A=Get IsEquipping(), B=NOT Boolean(A=@n1)))
  ]
 }
 Transition {
  path=AnimGraph.AnimGraphNode_StateMachine_0.AloeCrouton_SM.AnimStateTransitionNode_24.Transition
  nodes [
   n1 = Get IsEquipping()
   n2 = Get IsEquipped()
   Result(bCanEnterTransition=AND Boolean!(A=Get IsZiplining(), B=NOT Boolean(A=@n2), C:bool=NOT Boolean(A=@n1)) -> (ReturnValue))
  ]
 }
 Transition {
  path=AnimGraph.AnimGraphNode_StateMachine_0.AloeCrouton_SM.AnimStateTransitionNode_25.Transition
  nodes [
   n1 = Get IsSwimming() -> (IsSwimming:bool)
   n2 = NOT Boolean(A=Get IsSprinting() -> (IsSprinting:bool))
   n3 = NOT Boolean(A=Get IsEquipping())
   n4 = Get IsSwimming()
   n5 = Get IsEquipped()
   n6 = Get IsBothHandFKOff() -> (IsBothHandFKOff:bool)
   Result(bCanEnterTransition=OR Boolean(A=AND Boolean!(A=@n4, B=@n2, C=@n3, D:bool=@n5, E:bool=@n6) -> (ReturnValue), B=NOT Boolean(A=@n1)))
  ]
 }
 Transition {
  path=AnimGraph.AnimGraphNode_StateMachine_0.AloeCrouton_SM.AnimStateTransitionNode_26.Transition
  nodes [
   n1 = NOT Boolean(A=Get IsEquipping())
   n2 = NOT Boolean(A=Get IsEquipped())
   n3 = Get IsSwimming()
   Result(bCanEnterTransition=OR Boolean(A=Get IsSwimSprinting() -> (IsSwimSprinting:bool), B=AND Boolean!(A=@n3, B=@n2, C=@n1) -> (ReturnValue)))
  ]
 }
 Transition {
  path=AnimGraph.AnimGraphNode_StateMachine_0.AloeCrouton_SM.AnimStateTransitionNode_27.Transition
  nodes [Result(bCanEnterTransition=Get IsEquipping())]
 }
 Transition {
  path=AnimGraph.AnimGraphNode_StateMachine_0.AloeCrouton_SM.AnimStateTransitionNode_3.Transition
  nodes [
   n1 = Get IsSwimSprinting()
   n2 = NOT Boolean(A=Get IsZiplining())
   n3 = AND Boolean!(A=@n2, B=Get IsHolstered()) -> (ReturnValue)
   n4 = AND Boolean!(A=NOT Boolean(A=Get IsEquipped()), B=NOT Boolean(A=Get IsEquipping()), C:bool=@n2) -> (ReturnValue)
   n5 = Get IsGettingInProp() -> (IsGettingInProp:bool)
   Result(bCanEnterTransition=AND Boolean!(A=OR Boolean!(A=@n4, B=@n3, C:bool=@n5) -> (ReturnValue), B=NOT Boolean(A=@n1)) -> (ReturnValue))
  ]
 }
 Transition {
  path=AnimGraph.AnimGraphNode_StateMachine_0.AloeCrouton_SM.AnimStateTransitionNode_4.Transition
  nodes [
   n1 = Get IsGettingInProp()
   n2 = NOT Boolean(A=Get IsHolstered())
   n3 = Get IsEquipping()
   Result(bCanEnterTransition=AND Boolean(A=OR Boolean!(A=@n3, B=@n2) -> (ReturnValue), B=NOT Boolean(A=@n1)))
  ]
 }
 Transition {path=AnimGraph.AnimGraphNode_StateMachine_0.AloeCrouton_SM.AnimStateTransitionNode_5.Transition, nodes [Result()]}
 Transition {
  path=AnimGraph.AnimGraphNode_StateMachine_0.AloeCrouton_SM.AnimStateTransitionNode_6.Transition
  nodes [Result(bCanEnterTransition="Not Equal (Float)"(A=Get Charge Count()))]
 }
 Transition {
  path=AnimGraph.AnimGraphNode_StateMachine_0.AloeCrouton_SM.AnimStateTransitionNode_7.Transition
  nodes [Result(bCanEnterTransition="float > float"(A=Get Charge Count()))]
 }
 Transition {path=AnimGraph.AnimGraphNode_StateMachine_0.AloeCrouton_SM.AnimStateTransitionNode_8.Transition, nodes [Result()]}
 AnimationTransitionGraph_0 {
  path=AnimGraph.AnimGraphNode_StateMachine_0.AloeCrouton_SM.AnimStateTransitionNode_9.AnimationTransitionGraph_0
  nodes [Result(bCanEnterTransition=Get IsCharging())]
 }
 EventGraph {nodes [
   Event BlueprintUpdateAnimation()
   > n1 = Sequence!(execute) -> (then_0, then_1, then_2:Exec, then_3:Exec, then_4:Exec)
   n2 = TryGetFortPawn()
   > Is Valid(InputObject=@n2)
   > n3 = Cast To FortPlayerPawn(Object=@n2)
   > n4 = Set As Fort Player Pawn(As Fort Player Pawn:Fort Player Pawn Object Reference=@n3) -> (Output_Get:Fort Player Pawn Object Reference)
   n5 = GetAnimInstance(self=Get Mesh(self=@n4))
   > Set IsBothHandFKOff(IsBothHandFKOff:bool="float < float"(A=GetCurveValue(self=@n5, CurveName=BothHandFK), B=0.1)) -> (Output_Get:bool)
   Is Valid(exec=@n1.then_1, InputObject=Get As B Aloe Crouton Athena() -> (As B Aloe Crouton Athena:B Aloe Crouton Athena Object Reference))
   n6 = Get As B Aloe Crouton Athena()
   > Set IsCharging(IsCharging:bool="float > float"(A=GetChargePercent!(self:Fort Weapon Object Reference=@n6) -> (ReturnValue:float))) -> (Output_Get:bool)
   > Set IsEquipping(IsEquipping:bool=IsBeingEquipped!(self:Fort Weapon Object Reference=Get As B Aloe Crouton Athena()) -> (ReturnValue:bool)) -> (Output_Get:bool)
   > Set IsEquipped(IsEquipped:bool=IsEquipped!(self:Fort Weapon Object Reference=Get As B Aloe Crouton Athena()) -> (ReturnValue:bool)) -> (Output_Get:bool)
   > Set Charge Count(Charge Count:double=GetChargeCount!(self:B Aloe Crouton Athena Object Reference=Get As B Aloe Crouton Athena()) -> (ChargeCount:double)) -> (Output_Get:double)
   > Set IsHolstered(IsHolstered:bool=IsWeaponHolstered!(self:Fort Pawn Object Reference=Get As Fort Player Pawn() -> (As Fort Player Pawn:Fort Player Pawn Object Reference)) -> (ReturnValue:bool)) -> (Output_Get:bool)
   > n7 = IsZiplining!(execute, self:Fort Player Pawn Object Reference=Get As Fort Player Pawn()) -> (then, ReturnValue:bool)
   > Set IsZiplining(IsZiplining:bool=@n7) -> (Output_Get:bool)
   > Set IsSwimming(IsSwimming:bool=IsSurfaceSwimming!(self:Fort Pawn Object Reference=Get As Fort Player Pawn()) -> (ReturnValue:bool)) -> (Output_Get:bool)
   > Set IsSprinting(IsSprinting:bool=IsSprinting!(self:Fort Pawn Object Reference=Get As Fort Player Pawn()) -> (ReturnValue:bool)) -> (Output_Get:bool)
   > Set IsSwimSprinting(IsSwimSprinting:bool=AND Boolean(A=Get IsSwimming(), B=Get IsSprinting())) -> (Output_Get:bool)
   > Set IsGettingInProp(IsGettingInProp:bool=HasMatchingGameplayTag!(
     self:Gameplay Tag Asset Interface Interface=Get As Fort Player Pawn()
     TagToCheck:Gameplay Tag Structure=(TagName="Granted.Athena.EnvItem.HidingProp.GettingInProp")
    ) -> (ReturnValue:bool)) -> (Output_Get:bool)
   Event BlueprintInitializeAnimation()
   n8 = GetOwningActor()
   > Is Valid(InputObject=@n8)
   > n9 = Cast To B_AloeCrouton_Athena!(execute, Object=@n8) -> (then, CastFailed, AsB Aloe Crouton Athena:B Aloe Crouton Athena Object Reference) {is=title K2Node_DynamicCast}
   > Set As B Aloe Crouton Athena(As B Aloe Crouton Athena:B Aloe Crouton Athena Object Reference=@n9) -> (Output_Get:B Aloe Crouton Athena Object Reference)
  ]}
]
variables [
 As B Aloe Crouton Athena {type=B_AloeCrouton_Athena_C, default=null}
 As Fort Player Pawn {type=FortPlayerPawn}
 Charge Count {type=number}
 IsBothHandFKOff {type=boolean}
 IsCharging {type=boolean}
 IsEquipped {type=boolean}
 IsEquipping {type=boolean}
 IsGettingInProp {type=boolean}
 IsHolstered {type=boolean}
 IsRiding {type=boolean}
 IsSprinting {type=boolean}
 IsSwimming {type=boolean}
 IsSwimSprinting {type=boolean}
 IsZiplining {type=boolean}
]
