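# Animation Blueprint for the AloeCrouton weapon's backpack mesh, playing its equip and unequip animations in step with the owning player's state.
# A state machine starts in a hidden-pose Wait Equip state and branches to Equip, Unequip, or horizontal and vertical zipline states.
# Each update reads pawn gameplay tags, ziplining, and weapon hidden and equipping flags to drive transitions.
# IsEquip also gates transitions but is not set here, while IsSwimming is set but unused.
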AnimBlueprint /AloeCroutonWeapon/Items/AloeCrouton/Animation/AloeCrouton_Backpack_AnimBP
target_skeleton=/AloeCroutonWeapon/Items/AloeCrouton_Backpack/Mesh/M_MED_AloeCrouton_Pack_Skeleton
graphs [
 AnimGraph {nodes [
   AnimGraphNode_StateMachine_0 = AloeCrouton_Backpack_SM() {is=title AnimGraphNode_StateMachine}
   Output Pose(Result=@AnimGraphNode_StateMachine_0) {layerGroup=None}
  ]}
 AloeCrouton_Backpack_SM {
  entry=Wait Equip
  path=AnimGraph.AnimGraphNode_StateMachine_0.AloeCrouton_Backpack_SM
  ids=id
  nodes [
   AnimStateNode_0 = Equip() {is=title AnimStateNode}
   AnimStateNode_1 = Unequip() {is=title AnimStateNode}
   AnimStateNode_3 = Wait Equip() {is=title AnimStateNode}
   AnimStateNode_4 = Horizontal Ziplining() {is=title AnimStateNode}
   AnimStateNode_6 = Vertical Ziplining() {is=title AnimStateNode}
   AnimStateTransitionNode_0 = AnimStateTransitionNode() {crossfadeDuration=0}
   AnimStateTransitionNode_1 = AnimStateTransitionNode() {crossfadeDuration=0}
   AnimStateTransitionNode_10 = AnimStateTransitionNode() {crossfadeDuration=0}
   AnimStateTransitionNode_12 = AnimStateTransitionNode() {crossfadeDuration=0}
   AnimStateTransitionNode_13 = AnimStateTransitionNode() {crossfadeDuration=0}
   AnimStateTransitionNode_15 = AnimStateTransitionNode() {crossfadeDuration=0}
   AnimStateTransitionNode_16 = AnimStateTransitionNode() {crossfadeDuration=0}
   AnimStateTransitionNode_2 = AnimStateTransitionNode() {crossfadeDuration=0}
   AnimStateTransitionNode_3 = AnimStateTransitionNode() {bAutomaticRuleBasedOnSequencePlayerInState=true, crossfadeDuration=0}
   AnimStateTransitionNode_4 = AnimStateTransitionNode() {crossfadeDuration=0}
   AnimStateTransitionNode_5 = AnimStateTransitionNode()
   AnimStateTransitionNode_8 = AnimStateTransitionNode() {crossfadeDuration=0}
  ]
  links [
   {
    dst=Unequip
    rule=AnimGraph.AnimGraphNode_StateMachine_0.AloeCrouton_Backpack_SM.AnimStateTransitionNode_0.Transition
    src=Equip
    via=AnimStateTransitionNode_0
   }
   {
    dst=Wait Equip
    rule=AnimGraph.AnimGraphNode_StateMachine_0.AloeCrouton_Backpack_SM.AnimStateTransitionNode_3.Transition
    src=Equip
    via=AnimStateTransitionNode_3
   }
   {
    dst=Wait Equip
    rule=AnimGraph.AnimGraphNode_StateMachine_0.AloeCrouton_Backpack_SM.AnimStateTransitionNode_4.Transition
    src=Equip
    via=AnimStateTransitionNode_4
   }
   {
    dst=Unequip
    rule=AnimGraph.AnimGraphNode_StateMachine_0.AloeCrouton_Backpack_SM.AnimStateTransitionNode_16.Transition
    src=Horizontal Ziplining
    via=AnimStateTransitionNode_16
   }
   {
    dst=Wait Equip
    rule=AnimGraph.AnimGraphNode_StateMachine_0.AloeCrouton_Backpack_SM.AnimStateTransitionNode_8.Transition
    src=Horizontal Ziplining
    via=AnimStateTransitionNode_8
   }
   {
    dst=Equip
    rule=AnimGraph.AnimGraphNode_StateMachine_0.AloeCrouton_Backpack_SM.AnimStateTransitionNode_1.Transition
    src=Unequip
    via=AnimStateTransitionNode_1
   }
   {
    dst=Wait Equip
    rule=AnimGraph.AnimGraphNode_StateMachine_0.AloeCrouton_Backpack_SM.AnimStateTransitionNode_2.Transition
    src=Unequip
    via=AnimStateTransitionNode_2
   }
   {
    dst=Wait Equip
    rule=AnimGraph.AnimGraphNode_StateMachine_0.AloeCrouton_Backpack_SM.AnimStateTransitionNode_12.Transition
    src=Vertical Ziplining
    via=AnimStateTransitionNode_12
   }
   {
    dst=Equip
    rule=AnimGraph.AnimGraphNode_StateMachine_0.AloeCrouton_Backpack_SM.AnimStateTransitionNode_5.Transition
    src=Wait Equip
    via=AnimStateTransitionNode_5
   }
   {
    dst=Horizontal Ziplining
    rule=AnimGraph.AnimGraphNode_StateMachine_0.AloeCrouton_Backpack_SM.AnimStateTransitionNode_15.Transition
    src=Wait Equip
    via=AnimStateTransitionNode_15
   }
   {
    dst=Unequip
    rule=AnimGraph.AnimGraphNode_StateMachine_0.AloeCrouton_Backpack_SM.AnimStateTransitionNode_13.Transition
    src=Wait Equip
    via=AnimStateTransitionNode_13
   }
   {
    dst=Vertical Ziplining
    rule=AnimGraph.AnimGraphNode_StateMachine_0.AloeCrouton_Backpack_SM.AnimStateTransitionNode_10.Transition
    src=Wait Equip
    via=AnimStateTransitionNode_10
   }
  ]
 }
 Equip {
  path=AnimGraph.AnimGraphNode_StateMachine_0.AloeCrouton_Backpack_SM.AnimStateNode_0.Equip
  nodes [
   n1 = Sequence Player 'AloeCrouton_Equip_Male_Backpack'() {
    bLoopAnimation=false
    groupName=backpack_equip
    method=SyncGroup
    sequence=/AloeCroutonWeapon/Items/AloeCrouton/Animation/Equip/AloeCrouton_Equip_Male_Backpack
   }
   Output Animation Pose(Result=@n1)
  ]
 }
 Unequip {
  path=AnimGraph.AnimGraphNode_StateMachine_0.AloeCrouton_Backpack_SM.AnimStateNode_1.Unequip
  nodes [Output Animation Pose(Result=Sequence Player 'AloeCrouton_Unequip_Male_Backpack'() {bLoopAnimation=false, sequence=/AloeCroutonWeapon/Items/AloeCrouton/Animation/Equip/AloeCrouton_Unequip_Male_Backpack})]
 }
 Wait Equip {
  path=AnimGraph.AnimGraphNode_StateMachine_0.AloeCrouton_Backpack_SM.AnimStateNode_3.Wait Equip
  nodes [Output Animation Pose(Result=Sequence Player 'AloeCrouton_Equip_Male_Backpack_HIdden_Pose'() {bLoopAnimation=false, sequence=/AloeCroutonWeapon/Items/AloeCrouton/Animation/Equip/AloeCrouton_Equip_Male_Backpack_HIdden_Pose})]
 }
 Horizontal Ziplining {
  path=AnimGraph.AnimGraphNode_StateMachine_0.AloeCrouton_Backpack_SM.AnimStateNode_4.Horizontal Ziplining
  nodes [Output Animation Pose(Result=Sequence Player 'AloeCrouton_Equip_Male_Backpack_HIdden_Pose'() {bLoopAnimation=false, sequence=&AloeCrouton_Equip_Male_Backpack_HIdden_Pose})]
 }
 Vertical Ziplining {
  path=AnimGraph.AnimGraphNode_StateMachine_0.AloeCrouton_Backpack_SM.AnimStateNode_6.Vertical Ziplining
  nodes [
   Sequence Player 'AloeCrouton_Equip_Male_Backpack_HIdden_Pose'() {bLoopAnimation=false, sequence=&AloeCrouton_Equip_Male_Backpack_HIdden_Pose}
   Output Animation Pose(Result=Sequence Player 'AloeCrouton_Unequip_Male_Backpack'() {bLoopAnimation=false, sequence=&AloeCrouton_Unequip_Male_Backpack})
  ]
 }
 Transition {
  path=AnimGraph.AnimGraphNode_StateMachine_0.AloeCrouton_Backpack_SM.AnimStateTransitionNode_0.Transition
  nodes [
   n1 = Get IsZiplining() -> (IsZiplining:bool)
   n2 = Get InVehicle() -> (InVehicle:bool)
   n3 = NOT Boolean(A=Get IsEquip() -> (IsEquip:bool))
   n4 = Get Actor Hidden in Game() -> (Actor Hidden in Game:bool)
   Result(bCanEnterTransition=AND Boolean!(A=OR Boolean(A=@n3, B=@n4), B=NOT Boolean(A=@n1), C:bool=NOT Boolean(A=@n2)) -> (ReturnValue))
  ]
 }
 Transition {
  path=AnimGraph.AnimGraphNode_StateMachine_0.AloeCrouton_Backpack_SM.AnimStateTransitionNode_1.Transition
  nodes [
   n1 = Get Actor Hidden in Game()
   n2 = Get IsHiding() -> (IsHiding:bool)
   Result(bCanEnterTransition=AND Boolean(A=Get IsEquip(), B=NOT Boolean(A=@n1), C=NOT Boolean(A=@n2)))
  ]
 }
 Transition {
  path=AnimGraph.AnimGraphNode_StateMachine_0.AloeCrouton_Backpack_SM.AnimStateTransitionNode_10.Transition
  nodes [Result(bCanEnterTransition=AND Boolean(A=Get Actor Hidden in Game(), B=Get IsZiplining(), C=Get IsEquip()))]
 }
 Transition {
  path=AnimGraph.AnimGraphNode_StateMachine_0.AloeCrouton_Backpack_SM.AnimStateTransitionNode_12.Transition
  nodes [
   n1 = Get Actor Hidden in Game()
   n2 = Get IsZiplining()
   Result(bCanEnterTransition=AND Boolean!(A=NOT Boolean(A=@n2), B=NOT Boolean(A=@n1)) -> (ReturnValue))
  ]
 }
 Transition {
  path=AnimGraph.AnimGraphNode_StateMachine_0.AloeCrouton_Backpack_SM.AnimStateTransitionNode_13.Transition
  nodes [
   n1 = Get IsZiplining()
   n2 = Get InVehicle()
   n3 = Get InVehicleSeatRoofed() -> (InVehicleSeatRoofed:bool)
   n4 = OR Boolean(
    A=Get InMotorcycle() -> (InMotorcycle:bool)
    B=AND Boolean!(A=NOT Boolean(A=@n1), B=NOT Boolean(A=@n2), C:bool=NOT Boolean(A=@n3)) -> (ReturnValue)
   )
   n5 = OR Boolean(A=NOT Boolean(A=Get IsEquip()), B=Get Actor Hidden in Game())
   n6 = NOT Boolean(A=Get InVehicleSeatRoofed())
   n7 = NOT Boolean(A=Get IsEquip())
   n8 = Get InVehicle()
   Result(bCanEnterTransition=OR Boolean!(A=AND Boolean(A=@n8, B=@n6, C=@n7), B=AND Boolean!(A=@n5, B=@n4) -> (ReturnValue), C:bool=Get IsHiding()) -> (ReturnValue))
  ]
 }
 Transition {
  path=AnimGraph.AnimGraphNode_StateMachine_0.AloeCrouton_Backpack_SM.AnimStateTransitionNode_15.Transition
  nodes [
   n1 = Get Actor Hidden in Game()
   Result(bCanEnterTransition=AND Boolean!(A=NOT Boolean(A=@n1), B=Get IsZiplining(), C:bool=Get IsEquip()) -> (ReturnValue))
  ]
 }
 Transition {
  path=AnimGraph.AnimGraphNode_StateMachine_0.AloeCrouton_Backpack_SM.AnimStateTransitionNode_16.Transition
  nodes [
   n1 = Get IsEquip()
   Result(bCanEnterTransition=OR Boolean!(A=NOT Boolean(A=@n1), B=Get Actor Hidden in Game()) -> (ReturnValue))
  ]
 }
 Transition {
  path=AnimGraph.AnimGraphNode_StateMachine_0.AloeCrouton_Backpack_SM.AnimStateTransitionNode_2.Transition
  nodes [
   n1 = NOT Boolean(A=Get IsHiding())
   n2 = NOT Boolean(A=Get IsEquipping() -> (IsEquipping:bool))
   n3 = NOT Boolean(A=Get InMotorcycle())
   n4 = Get InVehicle()
   n5 = Get IsEquip()
   Result(bCanEnterTransition=OR Boolean(A=AND Boolean!(A=@n4, B=@n5, C=@n2, D:bool=@n3, E:bool=@n1) -> (ReturnValue), B=Get InVehicleSeatRoofed()))
  ]
 }
 Transition {path=AnimGraph.AnimGraphNode_StateMachine_0.AloeCrouton_Backpack_SM.AnimStateTransitionNode_3.Transition, nodes [Result()]}
 Transition {
  path=AnimGraph.AnimGraphNode_StateMachine_0.AloeCrouton_Backpack_SM.AnimStateTransitionNode_4.Transition
  nodes [Result(bCanEnterTransition=NOT Boolean(A=Get IsEquipping()))]
 }
 Transition {
  path=AnimGraph.AnimGraphNode_StateMachine_0.AloeCrouton_Backpack_SM.AnimStateTransitionNode_5.Transition
  nodes [
   n1 = Get InVehicle()
   Result(bCanEnterTransition=AND Boolean!(A=NOT Boolean(A=@n1), B=Get IsEquip(), C=Get IsEquipping()) -> (ReturnValue))
  ]
 }
 Transition {
  path=AnimGraph.AnimGraphNode_StateMachine_0.AloeCrouton_Backpack_SM.AnimStateTransitionNode_8.Transition
  nodes [Result(bCanEnterTransition=NOT Boolean(A=Get IsZiplining()))]
 }
 EventGraph {nodes [
   Event BlueprintUpdateAnimation()
   > n1 = Sequence!(execute) -> (then_0, then_1, then_2:Exec, then_3:Exec)
   n2 = TryGetFortPawn()
   > Is Valid(InputObject=@n2)
   > n3 = Cast To FortPlayerPawn(Object=@n2)
   > Set As Fort Player Pawn(As Fort Player Pawn:Fort Player Pawn Object Reference=@n3) -> (Output_Get:Fort Player Pawn Object Reference)
   n4 = GetCurrentWeapon!(self:Fort Pawn Object Reference=@n2) -> (ReturnValue:Fort Weapon Object Reference)
   > Is Valid(InputObject=@n4)
   > n5 = Cast To B_AloeCrouton_Athena!(execute, Object=@n4) -> (then, CastFailed, AsB Aloe Crouton Athena:B Aloe Crouton Athena Object Reference) {is=title K2Node_DynamicCast}
   > Set As B Aloe Crouton Athena(As B Aloe Crouton Athena:B Aloe Crouton Athena Object Reference=@n5) -> (Output_Get:B Aloe Crouton Athena Object Reference)
   Is Valid(exec=@n1.then_1, InputObject=Get As Fort Player Pawn() -> (As Fort Player Pawn:Fort Player Pawn Object Reference))
   > n6 = IsZiplining!(execute, self:Fort Player Pawn Object Reference=Get As Fort Player Pawn()) -> (then, ReturnValue:bool)
   > Set IsZiplining(IsZiplining:bool=@n6) -> (Output_Get:bool)
   > Set InVehicle(InVehicle:bool=HasMatchingGameplayTag!(
     self:Gameplay Tag Asset Interface Interface=Get As Fort Player Pawn()
     TagToCheck:Gameplay Tag Structure=(TagName="Gameplay.Action.Player.AthenaInVehicle")
    ) -> (ReturnValue:bool)) -> (Output_Get:bool)
   > Set InVehicleSeatRoofed(InVehicleSeatRoofed:bool=HasMatchingGameplayTag(self=Get As Fort Player Pawn(), TagToCheck=(TagName="Gameplay.Status.InVehicle.Seat.Roofed"))) -> (Output_Get:bool)
   > Set InMotorcycle(InMotorcycle:bool=HasMatchingGameplayTag(self=Get As Fort Player Pawn(), TagToCheck=(TagName="Gameplay.Action.Player.InVehicle.Motorcycle"))) -> (Output_Get:bool)
   > Set IsHiding(IsHiding:bool=HasMatchingGameplayTag(self=Get As Fort Player Pawn(), TagToCheck=(TagName="Granted.Athena.EnvItem.HidingProp.GettingInProp"))) -> (Output_Get:bool)
   > Set IsSwimming(IsSwimming:bool=IsSurfaceSwimming!(self:Fort Pawn Object Reference=Get As Fort Player Pawn()) -> (ReturnValue:bool)) -> (Output_Get:bool)
   n7 = Get As B Aloe Crouton Athena() -> (As B Aloe Crouton Athena:B Aloe Crouton Athena Object Reference)
   Is Valid(exec=@n1.then_2, InputObject=@n7)
   > Set Actor Hidden in Game(Actor Hidden in Game:bool=Get bHidden!(self:Actor Object Reference=@n7) -> (bHidden:bool)) -> (Output_Get:bool)
   > Set IsEquipping(IsEquipping:bool=IsBeingEquipped!(self:Fort Weapon Object Reference=Get As B Aloe Crouton Athena()) -> (ReturnValue:bool)) -> (Output_Get:bool)
  ]}
]
variables [
 Actor Hidden in Game {type=boolean}
 As B Aloe Crouton Athena {type=B_AloeCrouton_Athena_C, default=null}
 As Fort Player Pawn {type=FortPlayerPawn}
 InMotorcycle {type=boolean}
 InVehicle {type=boolean}
 InVehicleSeatRoofed {type=boolean}
 IsEquip {type=boolean}
 IsEquipping {type=boolean}
 IsHiding {type=boolean}
 IsSwimming {type=boolean}
 IsZiplining {type=boolean}
]
