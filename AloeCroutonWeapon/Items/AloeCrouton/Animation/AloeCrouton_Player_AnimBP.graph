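# Player animation Blueprint for the AloeCrouton weapon, overriding item layers on the Fortnite player skeleton.
# Each update caches pawn and weapon state, easing an alpha toward one when equipping while falling, or ziplining.
# The final-pose state machine uses it to blend in an in-air equip animation with LowerBody masking, then passes through when playback ends, tactical sprint starts or the weapon hides.
# The upper-body layer adds a charge loop scaled by ChargeAlpha.
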AnimBlueprint /AloeCroutonWeapon/Items/AloeCrouton/Animation/AloeCrouton_Player_AnimBP
target_skeleton=&Fortnite_M_Avg_Player_Skeleton
graphs [
 AnimGraph {nodes [Output Pose() {layerGroup=None}]}
 EventGraph {nodes [
   n1 = Event BlueprintUpdateAnimation()
   > Set Delta Time X(Delta Time X=@n1.DeltaTimeX)
   > n2 = Sequence!(execute) -> (then_0, then_1, then_2:Exec, then_3:Exec)
   n3 = TryGetPawnOwner()
   > Is Valid(InputObject=@n3)
   > n4 = Cast To FortPlayerPawn(Object=@n3)
   > Set As Fort Player Pawn(As Fort Player Pawn:Fort Player Pawn Object Reference=@n4) -> (Output_Get:Fort Player Pawn Object Reference)
   n5 = GetCurrentWeapon!(self:Fort Pawn Object Reference=Get As Fort Player Pawn() -> (As Fort Player Pawn:Fort Player Pawn Object Reference)) -> (ReturnValue:Fort Weapon Object Reference)
   > Is Valid(InputObject=@n5)
   > n6 = Cast To B_AloeCrouton_Athena!(execute, Object=@n5) -> (then, CastFailed, AsB Aloe Crouton Athena:B Aloe Crouton Athena Object Reference) {is=title K2Node_DynamicCast}
   > Set As B Aloe Crouton Athena(As B Aloe Crouton Athena:B Aloe Crouton Athena Object Reference=@n6) -> (Output_Get:B Aloe Crouton Athena Object Reference)
   Is Valid(exec=@n2.then_1, InputObject=Get As Fort Player Pawn())
   > Is Valid(InputObject=Get As B Aloe Crouton Athena() -> (As B Aloe Crouton Athena:B Aloe Crouton Athena Object Reference))
   n7 = Get As Fort Player Pawn()
   > Set IsFalling(IsFalling:bool=IsFalling!(self:Nav Movement Component Object Reference=Get CharacterMovement!(self:Character Object Reference=@n7) -> (CharacterMovement:Character Movement Component Object Reference)) -> (ReturnValue:bool)) -> (Output_Get:bool)
   > Set IsEquipping(IsEquipping:bool=IsBeingEquipped!(self:Fort Weapon Object Reference=Get As B Aloe Crouton Athena()) -> (ReturnValue:bool)) -> (Output_Get:bool)
   > n8 = IsZiplining!(execute, self:Fort Player Pawn Object Reference=Get As Fort Player Pawn()) -> (then, ReturnValue:bool)
   > Set IsZiplining(IsZiplining:bool=@n8) -> (Output_Get:bool)
   > Set IsTacSprinting(IsTacSprinting:bool=HasMatchingGameplayTag!(
     self:Gameplay Tag Asset Interface Interface=Get As Fort Player Pawn()
     TagToCheck:Gameplay Tag Structure=(TagName="Gameplay.Action.Player.TacticalSprint")
    ) -> (ReturnValue:bool)) -> (Output_Get:bool)
   > Set Actor Hidden in Game(Actor Hidden in Game:bool=Get bHidden!(self:Actor Object Reference=Get As B Aloe Crouton Athena()) -> (bHidden:bool)) -> (Output_Get:bool)
   n9 = Get IsFalling() -> (IsFalling:bool)
   n10 = Get IsEquipping() -> (IsEquipping:bool)
   n11 = Set IsEquipFalling(execute=@n2.then_2, IsEquipFalling:bool=OR Boolean(A=AND Boolean(A=@n9, B=@n10), B=Get IsZiplining() -> (IsZiplining:bool))) -> (Output_Get:bool)
   n12 = "To Float (Boolean)"(InBool=@n11)
   > Set EquipFallingAdditiveAlpha(EquipFallingAdditiveAlpha:float=FInterpTo(
     Current=Get EquipFallingAdditiveAlpha() -> (EquipFallingAdditiveAlpha:float)
     Target="Clamp (Float)"(Value=@n12, Min=0.001, Max=1)
     DeltaTime=Get Delta Time X()
     InterpSpeed=6
    )) -> (Output_Get:float)
  ]}
 ItemAimOffset {nodes [
   Input Pose!() -> (Pose, AimOffsetAlpha:double, Yaw:double, Pitch:double) {name=InPose_AimOffset}
   Input Pose!() -> (Pose) {name=InPose_PreAimOffsets}
   Output Pose() {layerGroup=Item Override}
  ]}
 ItemFallAdditive {nodes [
   Input Pose() {name=InPoseItemFallAdditive}
   Output Pose() {layerGroup=Item Override}
  ]}
 ItemFallLandAdditive {nodes [
   Input Pose() {name=InPoseItemFallLandAdditive}
   Output Pose() {layerGroup=Item Override}
  ]}
 ItemFallLevel2Override {nodes [
   Input Pose() {name=InPose_FallLevel2}
   Output Pose() {layerGroup=Item Override}
  ]}
 ItemFinalPoseOverride {nodes [
   AnimGraphNode_StateMachine_0 = New State Machine() {is=title AnimGraphNode_StateMachine}
   Output Pose(Result=@AnimGraphNode_StateMachine_0) {layerGroup=Item Override}
   Save cached pose 'Cached_FinalPose'(Pose=Input Pose() {name=InFinalPose}) {cacheName=Cached_FinalPose}
  ]}
 New State Machine {
  entry=Equip
  path=ItemFinalPoseOverride.AnimGraphNode_StateMachine_0.New State Machine
  ids=id
  nodes [
   AnimStateNode_0 = Equip() {is=title AnimStateNode}
   AnimStateNode_1 = Wait() {is=title AnimStateNode}
   AnimStateTransitionNode_0 = AnimStateTransitionNode() {bAutomaticRuleBasedOnSequencePlayerInState=true}
   AnimStateTransitionNode_1 = AnimStateTransitionNode() {crossfadeDuration=0}
  ]
  links [
   {
    dst=Wait
    rule=ItemFinalPoseOverride.AnimGraphNode_StateMachine_0.New State Machine.AnimStateTransitionNode_0.Transition
    src=Equip
    via=AnimStateTransitionNode_0
   }
   {
    dst=Wait
    rule=ItemFinalPoseOverride.AnimGraphNode_StateMachine_0.New State Machine.AnimStateTransitionNode_1.Transition
    src=Equip
    via=AnimStateTransitionNode_1
   }
  ]
 }
 Equip {
  path=ItemFinalPoseOverride.AnimGraphNode_StateMachine_0.New State Machine.AnimStateNode_0.Equip
  nodes [
   n1 = Layered blend per bone(
    BasePose=Sequence Player 'AloeCrouton_Equip_Male_inAir'() {bLoopAnimation=false, sequence=/AloeCroutonWeapon/Items/AloeCrouton/Animation/Equip/AloeCrouton_Equip_Male_inAir}
    BlendPoses_0=Use cached pose 'Cached_FinalPose'()
   ) {
    bMeshSpaceScaleBlend=true
    blendMode=BlendMask
    blendMasks [&Fortnite_M_Avg_Player_Skeleton:LowerBody]
    layerSetup []
   }
   Output Animation Pose(Result=Two Way Blend(A=Use cached pose 'Cached_FinalPose'(), B=@n1, Alpha=Get EquipFallingAdditiveAlpha()) {blendNode {alphaBoolBlend {blendOption=Linear}}})
  ]
 }
 Wait {
  path=ItemFinalPoseOverride.AnimGraphNode_StateMachine_0.New State Machine.AnimStateNode_1.Wait
  nodes [Output Animation Pose(Result=Use cached pose 'Cached_FinalPose'())]
 }
 Transition {path=ItemFinalPoseOverride.AnimGraphNode_StateMachine_0.New State Machine.AnimStateTransitionNode_0.Transition, nodes [Result()]}
 Transition {
  path=ItemFinalPoseOverride.AnimGraphNode_StateMachine_0.New State Machine.AnimStateTransitionNode_1.Transition
  nodes [Result(bCanEnterTransition=OR Boolean(A=Get IsTacSprinting() -> (IsTacSprinting:bool), B=Get Actor Hidden in Game() -> (Actor Hidden in Game:bool)))]
 }
 ItemFullBodyAimOffsets {nodes [
   Input Pose!() -> (Pose, FullBody_AimOffsetAlpha:double, Aim_Yaw:double, Aim_Pitch:double) {name=InPoseFullBodyAimOffsets}
   Input Pose!() -> (Pose) {name=InPosePreFullBodyAimOffsets}
   Output Pose() {layerGroup=Item Override}
  ]}
 ItemFullBodyOverride {nodes [
   Input Pose() {name=InPoseFullBody}
   Output Pose() {layerGroup=Item Override}
  ]}
 ItemIdleOverride {nodes [
   Input Pose() {name=InPose_Idle}
   Output Pose() {layerGroup=Item Override}
  ]}
 ItemJetPackJumpAdditive {nodes [Output Pose() {layerGroup=Item Override}]}
 ItemJetPackStartAdditive {nodes [Output Pose() {layerGroup=Item Override}]}
 ItemJumpLoopAdditive {nodes [
   Input Pose() {name=InPoseItemJumpLoopAdditive}
   Output Pose() {layerGroup=Item Override}
  ]}
 ItemJumpOverride {nodes [
   Input Pose() {name=InPoseJump}
   Output Pose() {layerGroup=Item Override}
  ]}
 ItemJumpUpAdditive {nodes [
   Input Pose() {name=InPoseItemJumpUpAdditive}
   Output Pose() {layerGroup=Item Override}
  ]}
 ItemPreAimOffsets {nodes [
   Input Pose() {name=InPosePreAimOffsets}
   Output Pose() {layerGroup=Item Override}
  ]}
 ItemPreIK {nodes [
   Input Pose() {name=InPose_PreIK}
   Output Pose() {layerGroup=Item Override}
  ]}
 ItemRightHandIK {nodes [
   Input Pose() {name=InPosePreRightHandIK}
   Input Pose() {name=InPosePostRightHandIK}
   Output Pose() {layerGroup=Item Override}
  ]}
 ItemSkeletalControl {nodes [
   Input Pose()
   Output Pose() {layerGroup=Item Override}
  ]}
 ItemSwimJumpFallLoop {nodes [Output Pose() {layerGroup=Item Override}]}
 ItemSwimJumpSurfaceEnd {nodes [Output Pose() {layerGroup=Item Override}]}
 ItemSwimJumpSurfaceLoop {nodes [Output Pose() {layerGroup=Item Override}]}
 ItemUpperBody {nodes [
   n1 = Get WeaponChargeLoop() -> (WeaponChargeLoop:Anim Sequence Object Reference)
   Output Pose(Result=Apply Additive(
     Base=Input Pose!() -> (Pose, InputParam:Fort Anim Input Adjusted Aim Structure) {name=InPoseUpperBody}
     Additive=Sequence Player!(Sequence:Anim Sequence Base Object Reference=@n1) -> (Pose)
     Alpha=Get ChargeAlpha() -> (ChargeAlpha:float)
    ) {alphaBoolBlend {blendOption=Linear}}) {layerGroup=Item Override}
   Get WeaponChargeLoop()
   Get ChargeAlpha()
  ]}
 ItemZipLineStartAdditive {nodes [Output Pose() {layerGroup=Item Override}]}
]
variables [
 Actor Hidden in Game {type=boolean}
 As B Aloe Crouton Athena {type=B_AloeCrouton_Athena_C, default=null}
 As Fort Player Pawn {type=FortPlayerPawn}
 Delta Time X {type=number}
 EquipFallingAdditiveAlpha {type=number}
 EquipPosition {type=number}
 IsEquipFalling {type=boolean}
 IsEquipping {type=boolean}
 IsFalling {type=boolean}
 IsTacSprinting {type=boolean}
 IsZiplining {type=boolean}
]
