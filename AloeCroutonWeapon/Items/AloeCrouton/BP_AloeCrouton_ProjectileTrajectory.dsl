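; Projectile trajectory preview actor for the AloeCrouton weapon, showing an arc of spline meshes and a target marker at its end.
; BeginPlay hides the marker and pre-creates a pool of spline meshes sharing one dynamic fade material.
; While enabled, Tick asks the owner to update the trajectory through FortTrajectoryAbilityInterface.
; Setting the spline runs parent logic, feeds endpoints to the material, and shows the marker at the last point, yawed to aim.
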(event EventBeginPlay
  (bind _maxmeshcount (Variables|Default|GetMaxMeshCount))
  (Rendering|SetVisibility (Variables|Default|GetTarget) false false)
  (Utilities|Array|Resize (Variables|Trajectory|GetSplineMeshComponents) _maxmeshcount)
  (for _index (range _maxmeshcount)
    (bind _returnvalue (AddComponent|NODE_AddSplineMeshComponent-0 :bManualAttachment true))
    (Utilities|Array|SetArrayElem (Variables|Trajectory|GetSplineMeshComponents) _index _returnvalue)
    (if (== _index 0)
      (bind _returnvalue_2 (Rendering|Material|CreateDynamicMaterialInstance _returnvalue 0 0 "None" :!class /Script/Engine.PrimitiveComponent))
      (bind _output_get (Variables|VFX|SetSplineMID _returnvalue_2))
      (Rendering|Material|SetScalarParameterValue _output_get "StartDistance" (Variables|Default|GetSplineFadeStartDistance) :!class /Script/Engine.MaterialInstanceDynamic)
      (Rendering|Material|SetScalarParameterValue _output_get "FadeDistance" (Variables|Default|GetSplineFadeDistance) :!class /Script/Engine.MaterialInstanceDynamic)
      (else
        (Rendering|Material|SetMaterial _returnvalue 0 (Variables|VFX|GetSplineMID))))))

(event EventTick
  (params (DeltaSeconds :type Float%20%28single-precision%29))
  ;[1 1 1] Tick - Update trajectory
  (|CallParentFunction DeltaSeconds)
  (if (Variables|Default|GetShouldUpdate)
    (CallFunction|UpdatefromTrajectoryOwner)))

(event Trajectory|EventSetTrajectorySpline
  (params (SplinePoints :type Array%20of%20Vectors) (SplineTangents :type Array%20of%20Vectors))
  (bind _returnvalue (Utilities|Array|LastIndex SplinePoints))
  (bind _output (Utilities|Array|Get(aref) SplinePoints _returnvalue))
  (bind _target (Variables|Default|GetTarget))
  ;[1 1 1] SetTrajectorySpline
  (|CallParentFunction SplinePoints SplineTangents)
  (if (> (Utilities|Array|Length SplinePoints) 0)
    (Rendering|Material|SetVectorParameterValue (Variables|VFX|GetSplineMID) "StartLocation" (Math|Conversions|ToLinearColor(Vector) (Utilities|Array|Get(aref) SplinePoints)) :!class /Script/Engine.MaterialInstanceDynamic)
    (Rendering|Material|SetVectorParameterValue (Variables|VFX|GetSplineMID) "EndLocation" (Math|Conversions|ToLinearColor(Vector) _output) :!class /Script/Engine.MaterialInstanceDynamic)
    (Transformation|SetWorldLocation _target _output false false)
    (Transformation|SetWorldRotation _target (Math|Rotator|MakeRotator 0.0 0.0 (+ (.yaw (Variables|Default|GetAim%20Rotation)) (Variables|Default|GetDirectionalForwardAdjustment))) false true)
    (Rendering|SetVisibility _target true false)))

(fn SetShouldUpdateFromOwner
  (params (ShouldUpdate :type Boolean) (Owner :type Object%20Reference))
  (bind _output_get (Variables|Default|SetShouldUpdate ShouldUpdate))
  (if _output_get
    (Variables|Default|SetTrajectory%20Owner Owner)
    (else
      (Variables|Default|SetTrajectory%20Owner 0))))

(fn UpdateFromTrajectoryOwner
  (bind _asfort_trajectory_ability_interface (Utilities|Casting|CastToFortTrajectoryAbilityInterface (Variables|Default|GetTrajectory%20Owner)))
  (TrajectoryInterface|UpdateTrajectory _asfort_trajectory_ability_interface))

(fn ConstructionScript)
