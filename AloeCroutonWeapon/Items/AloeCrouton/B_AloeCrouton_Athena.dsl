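; Fortnite weapon Blueprint for the AloeCrouton chargeable rocket launcher, handling charge state, feedback and an aim trajectory.
; The server copies charging attribute set values into replicated variables, letting sim proxies read charge count and cooldown without an ASC.
; Charging applies slow tags, pulses a sound indicator and drives overheat material blends, with timelines and audio covering overheat and cooldown.
; Helper functions spawn a trajectory actor and shape its spline by charge percent.
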(event EventBeginPlay
  (bind (_charge_minimum _charge_maximum _charge_biscurrentclamped _charge_bisbaseclamped _charge_bshouldclampbase _charge_unclampedbasevalue _charge_unclampedcurrentvalue _charge_bshouldscalewithmax _charge_basevalue _charge_currentvalue) (Utilities|Struct|BreakFortGameplayAttributeData (Variables|/Script/AloeCroutonWeaponRuntime.FortChargingSet_AloeCrouton|GetCharge_Minimum (Variables|Default|GetChargingAttributeSet))))
  (bind (_maxcharge_minimum _maxcharge_maximum _maxcharge_biscurrentclamped _maxcharge_bisbaseclamped _maxcharge_bshouldclampbase _maxcharge_unclampedbasevalue _maxcharge_unclampedcurrentvalue _maxcharge_bshouldscalewithmax _maxcharge_basevalue _maxcharge_currentvalue) (Utilities|Struct|BreakFortGameplayAttributeData (Variables|/Script/AloeCroutonWeaponRuntime.FortChargingSet_AloeCrouton|GetMaxCharge_Minimum (Variables|Default|GetChargingAttributeSet))))
  (bind (_servertimechargeincrements_minimum _servertimechargeincrements_maximum _servertimechargeincrements_biscurrentclamped _servertimechargeincrements_bisbaseclamped _servertimechargeincrements_bshouldclampbase _servertimechargeincrements_unclampedbasevalue _servertimechargeincrements_unclampedcurrentvalue _servertimechargeincrements_bshouldscalewithmax _servertimechargeincrements_basevalue _servertimechargeincrements_currentvalue) (Utilities|Struct|BreakFortGameplayAttributeData (Variables|/Script/AloeCroutonWeaponRuntime.FortChargingSet_AloeCrouton|GetServerTimeChargeIncrements_Minimum (Variables|Default|GetChargingAttributeSet))))
  (bind (_chargerate_minimum _chargerate_maximum _chargerate_biscurrentclamped _chargerate_bisbaseclamped _chargerate_bshouldclampbase _chargerate_unclampedbasevalue _chargerate_unclampedcurrentvalue _chargerate_bshouldscalewithmax _chargerate_basevalue _chargerate_currentvalue) (Utilities|Struct|BreakFortGameplayAttributeData (Variables|/Script/AloeCroutonWeaponRuntime.FortChargingSet_AloeCrouton|GetChargeRate_Minimum (Variables|Default|GetChargingAttributeSet))))
  ;[1 1 1] TDiA: added this to tick the equip animation once at startup so that we don't see it pop the first time you equip
  (Animation|Utilities|ForceTickAnimation (Variables|Visual|GetWeaponMesh))
  (Networking|SwitchHasAuthority
    (:Authority
      ;[1 1 1] Cache attribute set
      (bind _returnvalue (GameplayAttributes|GetAttributeSet (Ability|GetAbilitySystemComponent (Game|GetInstigator)) "/Script/AloeCroutonWeaponRuntime.FortChargingSet_AloeCrouton"))
      (bind _asfort_charging_set_aloe_crouton (Utilities|Casting|CastToFortChargingSet_AloeCrouton _returnvalue))
      (Variables|Default|SetChargingAttributeSet _asfort_charging_set_aloe_crouton)
      (Utilities|FlowControl|Sequence
        (:then_0
          ;[1 1 1] Set and listen for attribute values so sim proxies can access the correct values (sim proxies don't have access to the ASC or attribute sets)
          (Variables|ReplicatedAttributes|SetReplicatedCharge _charge_currentvalue)
          (bind (_asynctaskproxy _attribute _newvalue) (Ability|Tasks|WaitForAttributeChanged (Game|GetInstigator) "(AttributeName=\"Charge\",Attribute=/Script/AloeCroutonWeaponRuntime.FortChargingSet_AloeCrouton:Charge,AttributeOwner=\"/Script/CoreUObject.Class'/Script/AloeCroutonWeaponRuntime.FortChargingSet_AloeCrouton'\")")
            (:Changed
              (Variables|ReplicatedAttributes|SetReplicatedCharge _newvalue))))
        (:then_1
          (Variables|ReplicatedAttributes|SetReplicatedMaxCharge _maxcharge_currentvalue))
        (:then_2
          (Variables|ReplicatedAttributes|SetReplicatedServerTimeIncrements _servertimechargeincrements_currentvalue)
          (bind (_asynctaskproxy_2 _attribute_2 _newvalue_2) (Ability|Tasks|WaitForAttributeChanged (Game|GetInstigator) "(AttributeName=\"ServerTimeChargeIncrements\",Attribute=/Script/AloeCroutonWeaponRuntime.FortChargingSet_AloeCrouton:ServerTimeChargeIncrements,AttributeOwner=\"/Script/CoreUObject.Class'/Script/AloeCroutonWeaponRuntime.FortChargingSet_AloeCrouton'\")")
            (:Changed
              (Variables|ReplicatedAttributes|SetReplicatedServerTimeIncrements _newvalue_2))))
        (:then_3
          (Variables|ReplicatedAttributes|SetReplicatedChargeRate _chargerate_currentvalue))))))

(event EventOnPlayWeaponFireFX
  (params (bPersistentFire :type Boolean) (bSecondaryFire :type Boolean))
  ;[1 1 1] Muzzle Flashes [Overcharged state currently not in use ]
  (|CallParentFunction bPersistentFire bSecondaryFire)
  (GameplayCue|ExecuteGameplayCueOnActor(Burst) (Game|GetInstigator) (Variables|Default|GetGameplay%20Cue%20Fire) (Variables|Default|GetParameters)))

(event EventOnChargeDown
  (Variables|Default|SetStarting%20Charge false)
  (CallFunction|RemoveSlow))

(event EventOnStartCharge
  ;[1 1 1] Setting a max charge variable and cleaning it up for weapon swap steam escaping audio
  (Default|CallUpdateUIonStartCharge)
  (CallFunction|BeginOverheat)
  (Variables|Default|SetMaxChargeForSwap false)
  (Variables|Default|SetAudioVisualizerEnabled true)
  ;[1 1 1] Audio Visualizer Timer and Cleanup
  (bind _returnvalue (Utilities|Time|SetTimerbyEvent (delegate Custom|ChargeAudioVisualizer) 0.75 true))
  (Variables|Default|SetAudioVisualizerTimer _returnvalue)
  (Variables|Default|SetStarting%20Charge true)
  (CallFunction|ChargeSlow)
  (CallFunction|HandleIdleLoopSound false 0.5 1.0))

(event Custom|ChargeAudioVisualizer
  ;[1 1 1] Audio Visualizer Timer and Cleanup
  (if (Variables|Default|GetAudioVisualizerEnabled)
    (Fortnite|ShowSoundIndicatorLocalforTeam :Location (Transformation|GetActorLocation) :MaxAudibleDistance 3e+03 :IndicatorType "Gunshot" :Team (Team|GetActorTeam self) :Affiliations (Utilities|Array|MakeArray "Hostile" "Neutral") :Tint "(R=0.000000,G=0.000000,B=0.000000,A=0.000000)" :OverrideIcon "/AloeCroutonWeapon/Icons/AV_AloeCrouton_32.AV_AloeCrouton_32" :AttenuationMaxRangeMultiplier 1.0 :AttenuationIntensityMultiplier 1.0)
    (else
      (Utilities|Time|ClearandInvalidateTimerbyHandle (Variables|Default|GetAudioVisualizerTimer)))))

(event EventOnLocalClientReachedMaxCharge
  ;[1 1 1] Setting a max charge variable and cleaning it up for weapon swap steam escaping audio
  (Variables|Default|SetMaxChargeForSwap true)
  (Variables|Default|SetStarting%20Charge false)
  (CallFunction|RemoveSlow))

(event EventOnEndCharge
  ;[1 1 1] Setting a max charge variable and cleaning it up for weapon swap steam escaping audio
  (Default|CallUpdateUIonEndCharge)
  (CallFunction|EndOverheat)
  (CallFunction|HandleIdleLoopSound true)
  (Variables|Default|SetMaxChargeForSwap false)
  (Variables|Default|SetAudioVisualizerEnabled false)
  (bind _returnvalue (CallFunction|GetChargePercent))
  ;[1 1 1] Audio Visualizer Timer and Cleanup
  (Development|PrintString (Utilities|String|ToString(Float) _returnvalue))
  (Variables|Default|SetStarting%20Charge false)
  (CallFunction|RemoveSlow))

(event FortWeapon|EventOnLocalUnEquip
  ;[1 1 1] Setting a max charge variable and cleaning it up for weapon swap steam escaping audio
  (|CallParentFunction)
  (if (Variables|Default|GetMaxChargeForSwap)
    (Audio|SpawnSoundAttached (Variables|Default|GetMaxChargeSwapAudio) (Variables|Visual|GetWeaponMesh))
    (Variables|Default|SetStarting%20Charge false)
    (CallFunction|RemoveSlow)))

(event EventOnPlayReloadFX
  (params (ReloadStage :type EFortReloadFXState%20Enum :default "ReloadStart"))
  (|CallParentFunction ReloadStage)
  (Variables|Default|SetStarting%20Charge false)
  (CallFunction|RemoveSlow))

(event Custom|BeginOverheat
  ;[1 1 1] Overheat Visuals
  (Networking|CanExecuteCosmeticEvents
    (:True
      (bind _returnvalue (Utilities|Time|SetTimerbyEvent (delegate Custom|UpdateOverheatVisuals) (Utilities|Time|GetWorldDeltaSeconds) true))
      (Variables|Default|SetOverheatVisualsTimer _returnvalue)
      (Variables|Default|SetWeaponMID (Class|BTechArtGlobalLibrary|GetDynamicMaterialInstance (Variables|Visual|GetWeaponMesh))))))

(event EventOnSetTargeting
  (params (bNewIsTargeting :type Boolean))
  ;[1 1 1] On ADS, start Targeting Trajectory.
  (bind _aloecustomtrajectory (Variables|Default|GetAloeCustomTrajectory)
    (:then
      (if (not (Variables|Default|GetbCharging))
        (Actor|DestroyActor _aloecustomtrajectory)))
    (:else)))

(event Custom|UpdateOverheatVisuals
  ;[1 1 1] Overheat Visuals
  (bind _returnvalue (CallFunction|GetChargePercent))
  (Rendering|Material|SetScalarParameterValue (Variables|Default|GetWeaponMID) "AC_OverheatBlend" _returnvalue :!class /Script/Engine.MaterialInstanceDynamic)
  (Default|CallUpdateUIOverheatVisuals _returnvalue)
  (if (== _returnvalue 1.0)
    (Utilities|Time|ClearandInvalidateTimerbyHandle (Variables|Default|GetOverheatVisualsTimer))))

(event Custom|EndOverheat
  (bind _chargecount (CallFunction|GetChargeCount))
  (bind _returnvalue (== _chargecount 0))
  (bind _weaponmid (Variables|Default|GetWeaponMID))
  ;[1 1 1] Overheat Visuals
  (Networking|CanExecuteCosmeticEvents
    (:True
      (Default|CallUpdateUIOverheatVisuals 0)
      (CallFunction|PlayOverheatAudio)
      (bind (_direction _chargepercent) (Timeline|TL_EndOverheat PlayFromStart)
        (:Update
          (Rendering|Material|SetScalarParameterValue _weaponmid "AC_OverheatBlend" _chargepercent :!class /Script/Engine.MaterialInstanceDynamic)
          (if _returnvalue
            (Rendering|Material|SetScalarParameterValue _weaponmid "AC_OverheatIntensity" _chargepercent :!class /Script/Engine.MaterialInstanceDynamic)
            (Rendering|Material|SetScalarParameterValue _weaponmid "AC_CooldownIntensity" (- 1.0 _chargepercent) :!class /Script/Engine.MaterialInstanceDynamic)
            (Components|Activation|Activate (Variables|Default|GetNS_AloeCrouton_RocketLauncher_RearMuzzle) false)
            (Variables|Audio|GetIdleAudio
              (:then)
              (:else))))
        (:Finished
          (if _returnvalue
            (Utilities|Time|SetTimerbyEvent (delegate Custom|UpdateCooldownVisuals) (Math|Float|Max(Float) (- (CallFunction|GetCooldownTimeRemaining) 3.0)))))))))

(event Custom|Remove%20Looping%20GC
  (Networking|SwitchHasAuthority
    (:Authority
      (GameplayCue|RemoveGameplayCue (Game|GetInstigator) (Variables|Default|GetGameplay%20Cue%20Tag)))))

(event Custom|UpdateCooldownVisuals
  (bind _weaponmid (Variables|Default|GetWeaponMID))
  ;[1 1 1] Overheat Visuals
  (bind (_direction _ac_cooldownintensity _ac_overheatintensity) (Timeline|TL_EndCooldown PlayFromStart)
    (:Update
      (Rendering|Material|SetScalarParameterValue _weaponmid "AC_CooldownIntensity" _ac_cooldownintensity :!class /Script/Engine.MaterialInstanceDynamic)
      (Rendering|Material|SetScalarParameterValue _weaponmid "AC_OverheatIntensity" _ac_overheatintensity :!class /Script/Engine.MaterialInstanceDynamic)
      (Components|Activation|Deactivate (Variables|Default|GetNS_AloeCrouton_RocketLauncher_RearMuzzle)))
    (:Finished
      (CallFunction|StopOverheatAudio))))

(event EventOnWeaponVisibilityChanged
  (params (bVisible :type Boolean) (bSetForLocalControllerOnly :type Boolean))
  ;[0 0 0] Idle audio logic, takes in account pawn being in a passenger seat
  (|CallParentFunction bVisible bSetForLocalControllerOnly)
  (bind _output_get (Variables|Audio|SetbIsVisible bVisible))
  (CallFunction|HandleIdleLoopSound (and _output_get (> (Math|Vector|VectorLength (Transformation|GetActorScale3D)) 1.0)))
  (if (Variables|Audio|GetbIsOverheat)
    (Variables|Audio|SetbCanPlayOverheat false)
    (else
      (Variables|Audio|SetbCanPlayOverheat true))))

(event Custom|HandleIdleLoopSound
  (params (bShouldPlay :type Boolean) (FadeInTime :type Float%20%28double-precision%29 :default 0.5) (FadeOutTime :type Float%20%28double-precision%29 :default 0.3))
  (bind _idleaudio (Variables|Audio|GetIdleAudio))
  ;[0 0 0] Idle audio logic, takes in account pawn being in a passenger seat
  (if bShouldPlay
    (if (Audio|Components|Audio|IsPlaying (Variables|Audio|GetIdleAudio))
      (Audio|Components|Audio|FadeOut _idleaudio FadeOutTime 0.0 "Linear" :!class /Script/Engine.AudioComponent))
    (Audio|Components|Audio|SetSound _idleaudio (select (Variables|Audio|GetbIsOverheat) (Variables|Audio|GetOverheatLoopSoundCue) (Variables|Audio|GetIdleLoopSoundCue)))
    (Audio|Components|Audio|FadeIn _idleaudio FadeInTime 1.0 0.0 "Linear" :!class /Script/Engine.AudioComponent)
    (bind _equipaudio (Variables|Default|GetEquipAudio)
      (:then
        (Audio|Components|Audio|Play _equipaudio 0.0))
      (:else))
    (else
      (Audio|Components|Audio|FadeOut (Variables|Audio|GetIdleAudio) FadeOutTime 0.0 "Linear" :!class /Script/Engine.AudioComponent)
      (bind _equipaudio_2 (Variables|Default|GetEquipAudio)
        (:then
          (Audio|Components|Audio|FadeOut _equipaudio_2 0.2 0.0 "Linear" :!class /Script/Engine.AudioComponent))
        (:else)))))

(event FortWeapon|EventOnEquippedAndAttached
  ;[0 0 0] Idle audio logic, takes in account pawn being in a passenger seat
  (bind _asplayer_pawn_athena_generic_parent (Utilities|Casting|CastToPlayerPawn_Athena_Generic_Parent (Game|GetInstigator)))
  (bind _output_get (Variables|Default|SetWeaponOwnerGeneric _asplayer_pawn_athena_generic_parent))
  (Default|BindEventtoOnExitedVehicleDispatcher _output_get (delegate Custom|OnExitedVehicle)))

(event Custom|OnExitedVehicle
  ;[0 0 0] Idle audio logic, takes in account pawn being in a passenger seat
  (if (not (Audio|Components|Audio|IsPlaying (Variables|Audio|GetIdleAudio)))
    (CallFunction|HandleIdleLoopSound true)))

(event FortWeapon|EventOnUnEquippedAndDetached
  ;[0 0 0] Idle audio logic, takes in account pawn being in a passenger seat
  (bind _weaponownergeneric (Variables|Default|GetWeaponOwnerGeneric)
    (:then
      (Default|UnbindEventfromOnExitedVehicleDispatcher _weaponownergeneric (delegate Custom|OnExitedVehicle)))
    (:else)))

(event Custom|PlayOverheatAudio
  ;[0 0 0] Overheat audio logic
  (if (== (CallFunction|GetChargeCount) 0)
    (Variables|Audio|SetbIsOverheat true)
    (if (Variables|Audio|GetbIsVisible)
      (CallFunction|HandleIdleLoopSound true)
      (if (Variables|Audio|GetbCanPlayOverheat)
        (Audio|SpawnSoundAttached (Variables|Audio|GetOverheatStartSoundCue) (Variables|Visual|GetWeaponMesh))
        (Variables|Audio|SetbCanPlayOverheat false)))))

(event Custom|StopOverheatAudio
  ;[0 0 0] Overheat audio logic
  (Variables|Audio|SetbIsOverheat false)
  (if (Variables|Audio|GetbIsVisible)
    (CallFunction|HandleIdleLoopSound true 0.2 0.2)
    (if (not (Variables|Audio|GetbCanPlayOverheat))
      (Audio|SpawnSoundAttached (Variables|Audio|GetOverheatEndSoundCue) (Variables|Visual|GetWeaponMesh))
      (Variables|Audio|SetbCanPlayOverheat true))))

(fn CacheProjectileVariables)

(fn CalculateProjectileVelocity
  (return (FortWeapon|GetProjectileSpeed (FortWeapon|GetChargePercent))))

(fn CalculateTrajectorySpline
  (bind (_startlocation _initialvelocity _projectilevelocity _outrotation) (CallFunction|SetupProjectileVariables))
  (FortWeapon|UpdateDamageStartLocation)
  (bind _playerpawn (Variables|Default|GetPlayerPawn)
    (:then
      (bind (_outhitresult _outsplinepoints _outsplinetangents) (Projectile|CalculateProjectileTrajectorySplinewithHitProfile _playerpawn _startlocation _initialvelocity _projectilevelocity (Variables|Default|GetfProjectile_GravityZ) (Variables|Default|GetfProjectile_Friction) (Variables|Default|GetfProjectile_Bounciness) (Variables|Default|GetfProjectile_Timestep) (Variables|Default|GetfProjectile_Extent) (Variables|Default|GetTrajectoryCollisionProfile) (Class|FortPawn|GetProjectileIgnoredActors _playerpawn) 1 20 (Math|Float|Lerp 3.5e+02 1e+03 (FortWeapon|GetChargePercent)) 3e+01))
      (bind _aloecustomtrajectory (Variables|Default|GetAloeCustomTrajectory)
        (:then
          (Trajectory|SetTrajectorySpline _aloecustomtrajectory _outsplinepoints _outsplinetangents)
          (Variables|/AloeCroutonWeapon/Items/AloeCrouton/BP_AloeCrouton_ProjectileTrajectory.BP_AloeCrouton_ProjectileTrajectory_C|SetAim%20Rotation _outrotation _aloecustomtrajectory)
          (return _outhitresult))
        (:else)))
    (:else)))

(fn ChargeSlow
  (if (Variables|Default|GetStarting%20Charge)
    (Ability|GameplayEffect|AddLooseGameplayTags (Game|GetInstigator) (Variables|Default|GetGameplay%20Tags_0))))

(fn CurrentScalarValue)

(fn GetChargePercent
  (bind _output_get (Variables|Default|SetChargePercent (FortWeapon|GetChargePercent)))
  (return _output_get))

(fn GetChargeCount
  (return (Variables|ReplicatedAttributes|GetReplicatedCharge)))

(fn GetChargeRate
  (return (Variables|ReplicatedAttributes|GetReplicatedChargeRate)))

(fn GetCooldownDuration
  (return (select (> (CallFunction|GetCooldownTimeRemaining) 0) (Variables|ReplicatedAttributes|GetReplicatedChargeRate) 0.0)))

(fn GetCooldownTimeRemaining
  (return (Math|Float|Max(Float) 0.0 (- (Variables|ReplicatedAttributes|GetReplicatedServerTimeIncrements) (Fortnite|Sync|Time|GetServerWorldTimeSeconds)))))

(fn GetMaxChargeCount
  (return (Variables|ReplicatedAttributes|GetReplicatedMaxCharge)))

(fn RemoveSlow
  (Ability|GameplayEffect|RemoveLooseGameplayTags (Game|GetInstigator) (Variables|Default|GetGameplay%20Tags_0)))

(fn SetDummyVariables
  (bind _trajectorydata (Class|FortKismetLibrary|GetProjectileTrajectoryData (Variables|Default|GetPlayerPawn) (Variables|Default|GetPrj_Dummy) 1.8e+04 0.1))
  (Variables|Default|SetDummyMaxSpeed (Utilities|Struct|BreakFortProjectileTrajectoryData _trajectorydata))
  (Variables|Default|SetDummyGravity (.GravityZ (Utilities|Struct|BreakFortProjectileTrajectoryData _trajectorydata)))
  (Variables|Default|SetDummyExtent (Math|Float|Max(Float) (.CapsuleRadius (Utilities|Struct|BreakFortProjectileTrajectoryData _trajectorydata)) (.CapsuleHalfHeight (Utilities|Struct|BreakFortProjectileTrajectoryData _trajectorydata))))
  (Variables|Default|SetDummyBouncieness (.Bounciness (Utilities|Struct|BreakFortProjectileTrajectoryData _trajectorydata)))
  (Variables|Default|SetDummyFriction (.Friction (Utilities|Struct|BreakFortProjectileTrajectoryData _trajectorydata)))
  (Variables|Default|SetDummyCollisionChannel (.CollisionObjectType (Utilities|Struct|BreakFortProjectileTrajectoryData _trajectorydata))))

(fn SetupProjectileTrajectory
  (bind _playerpawn (Variables|Default|GetPlayerPawn)
    (:then
      (bind _returnvalue (Game|SpawnActorfromClass (Variables|Default|GetTrajectoryClass) (Transformation|GetActorTransform _playerpawn) "AlwaysSpawn" "SelectDefaultAtRuntime" _playerpawn _playerpawn))
      (bind _asbp_aloe_crouton_projectile_trajectory (Utilities|Casting|CastToBP_AloeCrouton_ProjectileTrajectory _returnvalue))
      (bind _output_get (Variables|Default|SetAloeCustomTrajectory _asbp_aloe_crouton_projectile_trajectory))
      (Actor|Tick|AddTickPrerequisiteActor _output_get self))
    (:else)))

(fn SetupProjectileVariables
  (return (Transformation|GetSocketLocation (Variables|Visual|GetWeaponMesh) "Muzzle") (* (Math|Rotator|GetRotationXVector (Math|Rotator|MakeRotator (.roll (Math|Conversions|RotationFromXVector (FortWeapon|GetDamageStartLocation -1))) (+ (.pitch (Math|Conversions|RotationFromXVector (FortWeapon|GetDamageStartLocation -1))) (FortWeapon|GetProjectilePitchOffset)) (.yaw (Math|Conversions|RotationFromXVector (FortWeapon|GetDamageStartLocation -1))))) (Variables|Default|GetfProjectileVelocity)) (Variables|Default|GetfProjectileVelocity) (Math|Rotator|MakeRotator (.roll (Math|Conversions|RotationFromXVector (FortWeapon|GetDamageStartLocation -1))) (+ (.pitch (Math|Conversions|RotationFromXVector (FortWeapon|GetDamageStartLocation -1))) (FortWeapon|GetProjectilePitchOffset)) (.yaw (Math|Conversions|RotationFromXVector (FortWeapon|GetDamageStartLocation -1))))))

(fn SetupTrajectoryProjectile
  (bind _returnvalue (Game|SpawnActorfromClass (Variables|Default|GetTrajectoryClass) (Transformation|GetActorTransform (Variables|Default|GetPlayerPawn)) "AlwaysSpawn" "SelectDefaultAtRuntime" (Variables|Default|GetPlayerPawn) (Variables|Default|GetPlayerPawn)))
  (bind _asbp_aloe_crouton_projectile_trajectory (Utilities|Casting|CastToBP_AloeCrouton_ProjectileTrajectory _returnvalue))
  (Variables|/AloeCroutonWeapon/Items/AloeCrouton/BP_AloeCrouton_ProjectileTrajectory.BP_AloeCrouton_ProjectileTrajectory_C|SetTrajectory%20Owner self _asbp_aloe_crouton_projectile_trajectory)
  (bind _output_get (Variables|Default|SetAs%20BP%20Aloe%20Crouton%20Projectile%20Trajectory _asbp_aloe_crouton_projectile_trajectory))
  (Actor|Tick|AddTickPrerequisiteActor _output_get self))

(fn ConstructionScript)
