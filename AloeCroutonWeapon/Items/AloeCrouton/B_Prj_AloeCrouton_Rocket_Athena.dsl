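; Rocket projectile for the AloeCrouton launcher, with a maximum-fuse timer, cosmetic tracer and spin, and explosion feedback.
; BeginPlay starts a timer that calls FuseTimerMax, then feeds muzzle location, start time and velocity-scaled WPO duration to mesh and trail parameters for a fake tracer.
; Tick spins the mesh on cosmetic-capable instances.
; Exploding calls the parent, shows a sound indicator, stops trail and audio, and spawns a radial force actor.
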(event EventBeginPlay
  (bind _returnvalue (Utilities|Time|GetGameTimeinSeconds))
  (bind _mesh (Variables|Projectile_Rocket|GetMesh))
  (bind _trail (Variables|Default|GetTrail))
  ;[1 1 1] EVENT -> BeginPlay
  (Utilities|Time|SetTimerbyEvent (delegate Custom|MaxFuse) (ScalableFloat|GetValueatLevel (Variables|Default|GetInput)))
  (Networking|CanExecuteCosmeticEvents
    (:True
      ;[1 1 1] Animate fake tracer mesh
      (bind _asfort_player_pawn (Utilities|Casting|CastToFortPlayerPawn (Variables|Actor|GetInstigator)))
      (Utilities|FlowControl|Sequence
        (:then_0
          (Rendering|Material|SetVectorParameterValueonMaterials _mesh "BulletOrigin" (Effects|GetMuzzleLocation (Variables|/Script/FortniteGame.FortPawn|GetCurrentWeapon _asfort_player_pawn) -1))
          (Rendering|Material|SetScalarParameterValueonMaterials _mesh "StartTime" _returnvalue)
          (Rendering|Material|SetScalarParameterValueonMaterials _mesh "WPODuration" (Math|Float|MapRangeClamped (Math|Vector|VectorLength (Transformation|GetVelocity)) 5e+03 4e+04 0.1 0.25))
          (Effects|Components|ParticleSystem|SetVectorParameter (Variables|Default|GetTrail) "MuzzlePosition" (Effects|GetMuzzleLocation (Variables|/Script/FortniteGame.FortPawn|GetCurrentWeapon _asfort_player_pawn) -1)))
        (:then_1
          (Effects|Components|ParticleSystem|SetVectorParameter _trail "BulletOrigin" (Effects|GetMuzzleLocation (Variables|/Script/FortniteGame.FortPawn|GetCurrentWeapon _asfort_player_pawn) -1))
          (Rendering|Material|SetScalarParameterforCustomPrimitiveData _trail "StartTime" _returnvalue))))))

(event Custom|MaxFuse
  (CallFunction|FuseTimerMax))

(event Collision|EventActorBeginOverlap
  (params (OtherActor :type Actor%20Object%20Reference)))

(event EventOnExploded
  (params (HitActors :type Array%20of%20Actor%20Object%20References) (HitResults :type Array%20of%20Hit%20Result%20Structures))
  (|CallParentFunction HitActors HitResults)
  (Fortnite|ShowSoundIndicatorLocal :Location (Transformation|GetActorLocation) :MaxAudibleDistance 0.0 :IndicatorType "Gunshot" :Instigator self :Tint "(R=0.000000,G=0.000000,B=0.000000,A=0.000000)" :SoundBase "/AloeCroutonWeapon/Sound/AloeCrouton_Launcher_Prj_Explo_Cue.AloeCrouton_Launcher_Prj_Explo_Cue" :AttenuationMaxRangeMultiplier 1.0 :AttenuationIntensityMultiplier 1.0)
  (Components|Activation|Deactivate (Variables|Default|GetTrail))
  (Audio|Components|Audio|FadeOut (Variables|Default|GetRocket_AudioComponent) 0.0 0.0 "Linear" :!class /Script/Engine.AudioComponent)
  (Game|SpawnActorfromClass "/Game/Effects/Niagara/Blueprints/PhsyicsFields/FN_RadialForce.FN_RadialForce_C" (Transformation|GetActorTransform self) "Undefined" "MultiplyWithRoot" 0 0.2 0.75 3e+03 1.5 0.0 0))

(event EventTick
  (params (DeltaSeconds :type Float%20%28single-precision%29))
  (Networking|CanExecuteCosmeticEvents
    (:True
      (if (Variables|Default|GetIsNiagaraEnabled)
        (if (Components|Activation|IsActive (Variables|Default|GetTrail))
          _)
        (elif true
          _))
      (Variables|Effects|RocketTrail|SetRocketSpin (Math|Rotator|MakeRotator (* DeltaSeconds (Variables|Effects|RocketTrail|GetRocketSpinRate))))
      (Transformation|AddLocalRotation (Variables|Projectile_Rocket|GetMesh) (Variables|Effects|RocketTrail|GetRocketSpin) false false))))

(fn ConstructionScript
  (|CallParentFunction))
