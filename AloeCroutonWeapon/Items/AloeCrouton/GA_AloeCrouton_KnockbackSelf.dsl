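; Gameplay ability for the AloeCrouton weapon that knocks its user back opposite their aim, scaled by weapon charge.
; On the server, LaunchPlayer skips pawns in vehicles or with knockback immunity and launches only falling ones, using scalable-float tuning for pitch, magnitude and Z, optionally resetting fall height on upward launches.
; A delayed timer then fires the knockback GameplayCue if absolute vertical speed exceeds a threshold, and ends the ability.
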(event Ability|EventActivateAbility
  (bind _asfort_player_pawn (Utilities|Casting|CastToFortPlayerPawn (Ability|GetAvatarActorfromActorInfo)))
  (bind _output_get (Variables|Default|SetPlayerPawn _asfort_player_pawn))
  (bind _asb_aloe_crouton_athena (Utilities|Casting|CastToB_AloeCrouton_Athena (Variables|/Script/FortniteGame.FortPawn|GetCurrentWeapon _output_get)))
  (bind _output_get_2 (Variables|Default|SetCharge%20Percent (FortWeapon|GetChargePercent _asb_aloe_crouton_athena)))
  (Development|PrintString (Utilities|String|ToString(Float) _output_get_2) true true "(R=0.000000,G=1.000000,B=0.081708,A=1.000000)")
  ;[1 1 1] Wait a brief moment (Like 2 frames) then check velocity. If Velocity on Z is high enough or low enough play GC.
  (CallFunction|LaunchPlayer)
  (Utilities|Time|SetTimerbyEvent (delegate Custom|CheckVel) (Variables|Default|GetFX%20Delay)))

(event Custom|CheckVel
  (bind _returnvalue (Ability|GetAvatarActorfromActorInfo))
  ;[1 1 1] Wait a brief moment (Like 2 frames) then check velocity. If Velocity on Z is high enough or low enough play GC.
  (Utilities|IsValid _returnvalue
    (:Is%20Valid
      (bind _asfort_player_pawn (Utilities|Casting|CastToFortPlayerPawn _returnvalue))
      (Development|PrintString (Utilities|String|ToString(Vector) (Transformation|GetVelocity _asfort_player_pawn)))
      (bind (_returnvalue_x _returnvalue_y _returnvalue_z) (Math|Vector|BreakVector (Transformation|GetVelocity _asfort_player_pawn)))
      (if (> (Math|Float|Absolute(Float) _returnvalue_z) (Variables|Default|GetVelocityForFX))
        (Ability|ExecuteGameplayCueOnOwner (Variables|Default|GetKnockbackCue))
        (else
          _))
      (Ability|EndAbility))))

(event Ability|EventFailedToActivatePassiveAbility
  (params (ActorInfo :type Gameplay%20Ability%20Actor%20Info%20Structure))
  (|CallParentFunction))

(fn LaunchPlayer
  (bind _returnvalue (Ability|GetAvatarActorfromActorInfo))
  (bind _finallaunchvel (Variables|Default|GetFinalLaunchVel))
  ;[1 1 1] Cast to pawn and launch the player a little bit. The distance launcged is based on a max distance from the player who hit them. If the player is further than the max distance than no launch is applied. Otherwise the player is launched based on a ratio of the knock mag using the distance from the player who launched them.
  (if (Networking|IsServer)
    (Utilities|IsValid _returnvalue
      (:Is%20Valid
        (bind _asfort_player_pawn (Utilities|Casting|CastToFortPlayerPawn _returnvalue))
        ;;[1 1 1] Check for Knock Immune or In Vehicle
        (if (not (or (Vehicle|IsinVehicle _asfort_player_pawn) (GameplayTags|HasMatchingGameplayTag _asfort_player_pawn "(TagName=\"Gameplay.Status.KnockbackImmunity\")")))
          ;;[1 1 1] If Crouched do different knock, but not if they are falling, if they are falling do the normal one.
          (if (and (Utilities|Enum|NotEqual(Enum) (Variables|/Script/Engine.CharacterMovementComponent|GetMovementMode (Variables|/Script/Engine.Character|GetCharacterMovement _asfort_player_pawn)) "MOVE_Falling") (Variables|/Script/Engine.Character|GetbIsCrouched _asfort_player_pawn))
            _
            (elif (Utilities|Enum|NotEqual(Enum) (Variables|/Script/Engine.CharacterMovementComponent|GetMovementMode (Variables|/Script/Engine.Character|GetCharacterMovement _asfort_player_pawn)) "MOVE_Falling")
              _
              (else
                (FortMovement|CancelJumpAbility _asfort_player_pawn false)
                ;;[1 1 1] Calc Launch Vel
                (Variables|Default|SetFinalLaunchVel (Math|Vector|MakeVector (.x (* (* (Math|Vector|Normalize (Math|Vector|GetForwardVector (Math|Rotator|MakeRotator (.roll (Pawn|GetBaseAimRotation _asfort_player_pawn)) (+ (.pitch (Pawn|GetBaseAimRotation _asfort_player_pawn)) (ScalableFloat|GetValueatLevel (Variables|Default|GetRow_OffsetPitch))) (.yaw (Pawn|GetBaseAimRotation _asfort_player_pawn))))) (* (ScalableFloat|GetValueatLevel (Variables|Default|GetRow_KnockMag)) (* (Variables|Default|GetCharge%20Percent) 0.8))) -1.0)) (.y (* (* (Math|Vector|Normalize (Math|Vector|GetForwardVector (Math|Rotator|MakeRotator (.roll (Pawn|GetBaseAimRotation _asfort_player_pawn)) (+ (.pitch (Pawn|GetBaseAimRotation _asfort_player_pawn)) (ScalableFloat|GetValueatLevel (Variables|Default|GetRow_OffsetPitch))) (.yaw (Pawn|GetBaseAimRotation _asfort_player_pawn))))) (* (ScalableFloat|GetValueatLevel (Variables|Default|GetRow_KnockMag)) (* (Variables|Default|GetCharge%20Percent) 0.8))) -1.0)) (Math|Float|float+float (.z (* (* (Math|Vector|Normalize (Math|Vector|GetForwardVector (Math|Rotator|MakeRotator (.roll (Pawn|GetBaseAimRotation _asfort_player_pawn)) (+ (.pitch (Pawn|GetBaseAimRotation _asfort_player_pawn)) (ScalableFloat|GetValueatLevel (Variables|Default|GetRow_OffsetPitch))) (.yaw (Pawn|GetBaseAimRotation _asfort_player_pawn))))) (* (ScalableFloat|GetValueatLevel (Variables|Default|GetRow_KnockMag)) (* (Variables|Default|GetCharge%20Percent) 0.8))) -1.0)) (ScalableFloat|GetValueatLevel (Variables|Default|GetRow_ZAdd)) (* (.z (Transformation|GetVelocity _asfort_player_pawn)) (ScalableFloat|GetValueatLevel (Variables|Default|GetRow_ZVelMult))))))
                ;;[1 1 1] Stop Movement and launch player
                (Components|Movement|StopMovementImmediately (Variables|/Script/Engine.Character|GetCharacterMovement _asfort_player_pawn))
                (Character|LaunchCharacter _asfort_player_pawn _finallaunchvel true true)
                ;;[1 1 1] If Velocity is Pos Z reset fall height
                (if (and (Math|Conversions|ToBoolean(Integer) (Math|Float|Floor (ScalableFloat|GetValueatLevel (Variables|Default|GetRow_ResetFall)))) (> (.z _finallaunchvel) 0.0))
                  (Pawn|Character|ResetFallingHeight _asfort_player_pawn)
                  (else
                    _)))))
          (Variables|Default|SetFinalLaunchVel (Math|Vector|MakeVector (.x (* (* (Math|Vector|Normalize (Math|Vector|GetForwardVector (Math|Rotator|MakeRotator (.roll (Pawn|GetBaseAimRotation _asfort_player_pawn)) 355.0 (.yaw (Pawn|GetBaseAimRotation _asfort_player_pawn))))) (* (ScalableFloat|GetValueatLevel (Variables|Default|GetRow_KnockMag)) 0.25)) -1.0)) (.y (* (* (Math|Vector|Normalize (Math|Vector|GetForwardVector (Math|Rotator|MakeRotator (.roll (Pawn|GetBaseAimRotation _asfort_player_pawn)) 355.0 (.yaw (Pawn|GetBaseAimRotation _asfort_player_pawn))))) (* (ScalableFloat|GetValueatLevel (Variables|Default|GetRow_KnockMag)) 0.25)) -1.0)) (Math|Float|float+float (.z (* (* (Math|Vector|Normalize (Math|Vector|GetForwardVector (Math|Rotator|MakeRotator (.roll (Pawn|GetBaseAimRotation _asfort_player_pawn)) 355.0 (.yaw (Pawn|GetBaseAimRotation _asfort_player_pawn))))) (* (ScalableFloat|GetValueatLevel (Variables|Default|GetRow_KnockMag)) 0.25)) -1.0)) (ScalableFloat|GetValueatLevel (Variables|Default|GetRow_ZAdd)) 0)))
          (Components|Movement|StopMovementImmediately (Variables|/Script/Engine.Character|GetCharacterMovement _asfort_player_pawn)))))))
