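; Passive ability for the AloeCrouton ("Lobster") weapon that handles in-water behavior, noted as keeping the Lobster holstered while in water.
; Tag-wait tasks watch Granted.Perk.ActionCharge.T01 on the player pawn as the water signal.
; On add it applies the GE_Slow variable's effect a tick later and stores the handle; on remove it clears that effect.
; Player death ends the ability, and ending also removes the effect.
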(event Ability|EventActivateAbility
  (bind _playerpawn (Variables|Default|GetPlayerPawn))
  (bind _asfort_player_pawn (Utilities|Casting|CastToFortPlayerPawn (Ability|GetAvatarActorfromActorInfo)))
  (Variables|Default|SetPlayerPawn _asfort_player_pawn)
  ;[1 1 1] End the ability when the player dies
  (EventDispatchers|AssignOnDied (Variables|Default|GetPlayerPawn) (delegate Custom|OnDied_Event_0))
  ;[1 1 1] In Water Behavior (keep the Lobster holstered, even while dolphin diving and water jumping)
  (Utilities|FlowControl|Sequence
    (:then_1
      (Ability|Tasks|WaitGameplayTagAdd "(TagName=\"Granted.Perk.ActionCharge.T01\")" _playerpawn
        (:Added
          (Utilities|FlowControl|DelayUntilNextTick)
          (Development|PrintString "Lobster In Water!" false false)
          (bind _returnvalue (Ability|ApplyGameplayEffectToOwner (Variables|Default|GetGE_Slow)))
          (Variables|Default|SetSlowHandle _returnvalue))))
    (:then_2
      (Ability|Tasks|WaitGameplayTagRemove "(TagName=\"Granted.Perk.ActionCharge.T01\")" _playerpawn
        (:Removed
          (Development|PrintString "Lobster Exit Water" false false)
          (Ability|RemoveGameplayEffectFromOwnerWithHandle (Variables|Default|GetSlowHandle)))))))

(event Custom|OnDied_Event_0
  (params (DamagedActor :type Actor%20Object%20Reference) (Damage :type Float%20%28single-precision%29) (InstigatedBy :type Controller%20Object%20Reference) (DamageCauser :type Actor%20Object%20Reference) (HitLocation :type Vector) (FHitComponent :type Primitive%20Component%20Object%20Reference) (BoneName :type Name) (Momentum :type Vector))
  ;[1 1 1] End the ability when the player dies
  (Ability|EndAbility))

(event Ability|EventOnEndAbility
  (params (bWasCancelled :type Boolean))
  ;[1 1 1] End Ability Cleanup
  (Ability|RemoveGameplayEffectFromOwnerWithHandle (Variables|Default|GetSlowHandle)))

(event Ability|EventFailedToActivatePassiveAbility
  (params (ActorInfo :type Gameplay%20Ability%20Actor%20Info%20Structure))
  (|CallParentFunction))
